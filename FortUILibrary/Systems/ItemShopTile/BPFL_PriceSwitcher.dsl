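; Blueprint function library for the item shop tile price switcher, supplying the real-price text that corresponds to an MTX price.
; GetRealPriceText requires a Ready store loading state, a positive price, a real price, Mtx currency and a showing price.
; When all conditions pass, it delegates formatting to the FortMtxStoreVM view model, forwarding the price and additional-text flag.
; Otherwise it returns empty text.
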(fn GetRealPriceText
  (params (FortMTXStoreVM :type Fort%20Mtx%20Store%20VM%20Object%20Reference) (LoadingState :type EFortMtxStoreLoadingState%20Enum :default "Uninitialized") (MtxPrice :type Integer) (HasRealPrice :type Boolean) (Currency :type EFortPriceCurrency%20Enum :default "Mtx") (IsPriceShowing :type Boolean) (IsAdditionalText :type Boolean))
  (if (Math|Boolean|ANDBoolean (Utilities|Enum|Equal(Enum) LoadingState "Ready") (> MtxPrice 0) (Variables|GetHasRealPrice) (Utilities|Enum|Equal(Enum) Currency "Mtx") (Variables|GetIsPriceShowing))
    (return (Class|FortMtxStoreVM|GetRealPriceText (Variables|GetFortMtxStoreVM) (Variables|GetMtxPrice) (Variables|GetIsAdditionalText)))
    (else
      (return ""))))
