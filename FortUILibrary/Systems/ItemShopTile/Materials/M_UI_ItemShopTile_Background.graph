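# UI material for item shop tile backgrounds, showing either a color gradient or a texture scaled to fill.
# The gradient blends two or three colors from a shared parameter function along a rotatable, sine-eased axis, normalized by |cos|+|sin| to span the tile at any angle.
# A static switch selects texture or gradient.
# A rounded-corners function takes tile-driven radius override parameters.
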Material /FortUILibrary/Systems/ItemShopTile/Materials/M_UI_ItemShopTile_Background
graph {
 properties {
  bUsedWithStaticMesh=false
  blendMode=BLEND_Translucent
  materialDomain=MD_UI
  pixelDepthOffsetMode=PDOM_Legacy
  usedShadingModels=Default Lit
 }
 nodes [
  n1 = VectorParameter() {group=Tile-driven params, parameterName=CornerRadiusOverride, sortPriority=40}
  MaterialFunctionCall!(
   Input Alpha=Constant() {r=1}
   bOverrideCornerRadius=ScalarParameter() {group=Tile-driven params, parameterName=bOverrideCornerRadius, sortPriority=30}
   Corner Radius Override=AppendVector(A=@n1.RGB, B=@n1.RGB)
  ) {materialFunction=/FortUILibrary/Systems/ItemShopTile/Materials/MaterialFunctions/MF_UI_ItemShop_RoundedCorners}
  n2 = MaterialFunctionCall() {materialFunction=/FortUILibrary/Systems/ItemShopTile/Materials/MaterialFunctions/MF_UI_ItemShop_BackgroundParams}
  NamedRerouteDeclaration(None=@n2.Color 1) {name=color 1, nodeColor=<rgba 0.2422 0 1 1>}
  NamedRerouteDeclaration(None=@n2.Color 2) {name=color 2, nodeColor=<rgba 0.0781 1 0 1>}
  NamedRerouteDeclaration(None=@n2.Color 3) {name=color 3, nodeColor=<rgba 1 0 0.3984 1>}
  n3 = NamedRerouteDeclaration(None=Multiply(A=ScalarParameter() {parameterName=Rotation, sortPriority=0}) {constB=-1}) {name=Rotation Angle, nodeColor=<rgba 1 0 0.5625 1>}
  n4 = Subtract(A=Rotator(Coordinate=MaterialFunctionCall() {materialFunction=&GetUserInterfaceUV}.Normalized UV, Time=@n3) {speed=6.2832}) {constB=0.5}
  n5 = Add(A=Abs(None=Cosine(None=@n3)), B=Abs(None=Sine(None=@n3)))
  n6 = ComponentMask(None=Add(A=Divide(A=@n4, B=@n5) {constB=1}) {constB=0.5}) {g=true}
  NamedRerouteDeclaration(None=MaterialFunctionCall!("Value (0-1)"=Saturate(None=@n6)) {materialFunction=/FortUILibrary/MaterialFunctions/Basics/MF_UI_EaseCurves_01_Sine}.Out) {name=Gradient, nodeColor=<rgba 0 0.7188 1 1>}
  n7 = TextureObjectParameter() {parameterName=Texture}
  n8 = TextureProperty(Tex=@n7)
  n9 = NamedRerouteUsage()
  n10 = NamedRerouteUsage()
  n11 = NamedRerouteUsage()
  n12 = NamedRerouteUsage()
  n13 = NamedRerouteUsage()
  n14 = NamedRerouteUsage()
  n15 = NamedRerouteUsage()
  StaticSwitchParameter(
   True=TextureSample(UVs=MaterialFunctionCall(TextureSize=@n8) {materialFunction=&MF_UI_UVScaling}.Scaled To Fill, Tex=@n7).RGB
   False=LinearInterpolate(
    A=LinearInterpolate(A=@n9, B=@n10, Alpha=@n11)
    B=MaterialFunctionCall(1=@n12, 2=@n13, 3=@n14, A=@n15) {materialFunction=&Lerp_Multiple_Float3}.Lerp 3 Inputs
    Alpha=ScalarParameter() {defaultValue=1, parameterName=bUse3Colors, sliderMax=1, sortPriority=0}
   )
  ) {parameterName=IsTextureBackground}
 ]
}
