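# Translucent UI material for Item Shop tile backgrounds, drawing either a radial color gradient or a texture.
# Two smoothstepped radial masks with adjustable position and scale, built on fill-scaled UVs, blend colors from a shared background-params function, with bUse3Colors interpolating between two blend variants.
# The IsTextureBackground static switch substitutes a texture scaled to fill.
# A rounded-corners function takes a tile-driven corner radius override.
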Material /FortUILibrary/Systems/ItemShopTile/Materials/M_UI_ItemShopTile_RadialBackground
graph {
 properties {
  bAutomaticallySetUsageInEditor=false
  bUsedWithStaticMesh=false
  blendMode=BLEND_Translucent
  materialDomain=MD_UI
  usedShadingModels=Default Lit
 }
 nodes [
  n1 = VectorParameter() {group=Tile-driven params, parameterName=CornerRadiusOverride, sortPriority=40}
  MaterialFunctionCall!(
   Input Alpha=Constant() {r=1}
   bOverrideCornerRadius=ScalarParameter() {group=Tile-driven params, parameterName=bOverrideCornerRadius, sortPriority=30}
   Corner Radius Override=AppendVector(A=@n1.RGB, B=@n1.RGB)
  ) {materialFunction=/FortUILibrary/Systems/ItemShopTile/Materials/MaterialFunctions/MF_UI_ItemShop_RoundedCorners}
  n2 = MaterialFunctionCall() {materialFunction=/FortUILibrary/Systems/ItemShopTile/Materials/MaterialFunctions/MF_UI_ItemShop_BackgroundParams}
  NamedRerouteDeclaration(None=@n2.Color 1) {name=color 1, nodeColor=<rgba 0.2422 0 1 1>}
  NamedRerouteDeclaration(None=@n2.Color 2) {name=color 2, nodeColor=<rgba 0.0781 1 0 1>}
  NamedRerouteDeclaration(None=@n2.Color 3) {name=color 3, nodeColor=<rgba 1 0 0.3984 1>}
  n3 = MaterialFunctionCall(UV=MaterialFunctionCall() {materialFunction=&GetUserInterfaceUV}.Normalized UV) {materialFunction=&MF_UI_UVScaling}
  n4 = Subtract(
   A=@n3.Scaled To Fill
   B=AppendVector(A=ScalarParameter() {parameterName=RadialGradient01PosX}, B=ScalarParameter() {parameterName=RadialGradient01PosY})
  )
  n5 = ScalarParameter() {defaultValue=1, parameterName=RadialGradient01Scale}
  n6 = Add(A=Divide(A=Subtract(A=@n4) {constB=0.5}, B=Max(A=@n5) {constB=0}) {constB=1}) {constB=0.5}
  n7 = Saturate(None=Length(None=ConstantBiasScale(None=@n6) {bias=-0.5, scale=2}) {is=class MaterialExpressionLength})
  NamedRerouteDeclaration(None=SmoothStep(Value=OneMinus(None=@n7)) {constMax=1.1}) {name=RadialGradient01, nodeColor=<rgba 0 0.7188 1 1>}
  n8 = Subtract(
   A=@n3.Scaled To Fill
   B=AppendVector(A=ScalarParameter() {parameterName=RadialGradient02PosX}, B=ScalarParameter() {parameterName=RadialGradient02PosY})
  )
  n9 = ScalarParameter() {defaultValue=1, parameterName=RadialGradient02Scale}
  n10 = Add(A=Divide(A=Subtract(A=@n8) {constB=0.5}, B=Max(A=@n9) {constB=0}) {constB=1}) {constB=0.5}
  n11 = Saturate(None=Length(None=ConstantBiasScale(None=@n10) {bias=-0.5, scale=2}))
  NamedRerouteDeclaration(None=SmoothStep(Value=OneMinus(None=@n11)) {constMax=1.2}) {name=RadialGradient02, nodeColor=<rgba 0 0.7188 1 1>}
  n12 = TextureObjectParameter() {parameterName=Texture}
  n13 = TextureProperty(Tex=@n12)
  n14 = NamedRerouteUsage()
  n15 = LinearInterpolate(A=NamedRerouteUsage(), B=@n14, Alpha=NamedRerouteUsage())
  n16 = NamedRerouteUsage()
  n17 = LinearInterpolate(A=NamedRerouteUsage(), B=NamedRerouteUsage(), Alpha=NamedRerouteUsage())
  n18 = NamedRerouteUsage()
  n19 = NamedRerouteUsage()
  StaticSwitchParameter(
   True=TextureSample(
    UVs=MaterialFunctionCall(TextureSize=@n13) {materialFunction=&MF_UI_UVScaling}.Scaled To Fill
    Tex=@n12
   ).RGB
   False=LinearInterpolate(
    A=LinearInterpolate(A=@n15, B=@n14, Alpha=@n16)
    B=LinearInterpolate(A=@n17, B=@n18, Alpha=@n19)
    Alpha=ScalarParameter() {defaultValue=1, parameterName=bUse3Colors, sliderMax=1, sortPriority=0}
   )
  ) {parameterName=IsTextureBackground}
 ]
}
