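# UI material for an item shop tile that draws a translucent two-color linear gradient with rounded corners.
# MF_UI_LinearGradient supplies a rotatable, offsettable gradient, smoothstepped to blend both the two colors and start/end alphas.
# The alpha passes through MF_UI_ItemShop_RoundedCorners, with radius overridden by MPC_ItemShop's CornerRadius.
# That radius is multiplied by bIndividualRoundedCorners, but only its TopLeft channel feeds all four components.
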Material /FortUILibrary/Systems/ItemShopTile/Materials/M_UI_MPShopTile_LinearGradient
graph {
 properties {
  bUsedWithStaticMesh=false
  blendMode=BLEND_Translucent
  materialDomain=MD_UI
  pixelDepthOffsetMode=PDOM_Legacy
  usedShadingModels=Default Lit
 }
 nodes [
  n1 = VectorParameter() {
   group=Gradient Settings
   parameterName=GradientRotationCenter
   sortPriority=5
   channelNames {a="", b="", g=Center Y, r=Center X}
   defaultValue=<rgba 0.5 0.5 0 0>
  }
  n2 = SmoothStep(Value=MaterialFunctionCall(
    GradientStart=ScalarParameter() {group=Gradient Settings, parameterName=GradientStart, sortPriority=1}
    GradientEnd=ScalarParameter() {defaultValue=1, group=Gradient Settings, parameterName=GradientEnd, sortPriority=2}
    RotationAngle=ScalarParameter() {group=Gradient Settings, parameterName=GradientRotation, sortPriority=4}
    RotationCenter=@n1
    GradientOffset=ScalarParameter() {group=Gradient Settings, parameterName=GradientOffset, sortPriority=3}
   ) {materialFunction=&MF_UI_LinearGradient}.Result)
  LinearInterpolate(
   A=VectorParameter() {group=Color Settings, parameterName=Color, defaultValue=<rgba 0 0 0 1>}.RGB
   B=VectorParameter() {group=Color Settings, parameterName=Color2, defaultValue=<rgba 1 1 1 1>}.RGB
   Alpha=@n2
  )
  n3 = VectorParameter() {parameterName=bIndividualRoundedCorners, channelNames {a=BottomRight, b=TopRight, g=BottomLeft, r=TopLeft}, defaultValue=<rgba 1 1 1 1>}
  MaterialFunctionCall!(
   Input Alpha=LinearInterpolate(
    A=ScalarParameter() {group=Gradient Settings, parameterName=AlphaStart, sortPriority=1}
    B=ScalarParameter() {defaultValue=1, group=Gradient Settings, parameterName=AlphaEnd, sortPriority=2}
    Alpha=@n2
   )
   bOverrideCornerRadius=Constant() {r=1}
   Corner Radius Override=Multiply(
    A=CollectionParameter() {collection=/FortUILibrary/Systems/ItemShopTile/MPC_ItemShop, parameterName=CornerRadius}
    B=MaterialFunctionCall(R=@n3.TopLeft, G=@n3.TopLeft, B=@n3.TopLeft, A=@n3.TopLeft) {materialFunction=&AppendMany}.RGBA
   )
  ) {materialFunction=/FortUILibrary/Systems/ItemShopTile/Materials/MaterialFunctions/MF_UI_ItemShop_RoundedCorners}
 ]
}
