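# Translucent UI material for an item shop tile placeholder, switching between an animated loading state and an empty-image state through the "Is Empty" scalar.
# A delayed, sine-eased animation drives a linear gradient into a Loading Mask for the pulsing or sweeping loading look.
# The empty state draws an image-error SDF icon, scaled to fit and anti-aliased, as a grayscale thumbnail.
# Alpha passes through the shop's rounded-corners function.
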Material /FortUILibrary/Systems/ItemShopTile/Materials/M_UI_Tile_Loading
graph {
 properties {
  bUsedWithStaticMesh=false
  blendMode=BLEND_Translucent
  materialDomain=MD_UI
  pixelDepthOffsetMode=PDOM_Legacy
  usedShadingModels=Default Lit
 }
 nodes [
  n1 = ScalarParameter() {defaultValue=1, parameterName=Is Empty, sliderMax=1}
  LinearInterpolate(A=Constant() {desc=loading color, r=0.35}, B=NamedRerouteUsage(), Alpha=@n1)
  n2 = Constant() {desc=loading min alpha, r=0.2}
  n3 = Constant() {desc=loading max alpha, r=0.4}
  n4 = NamedRerouteUsage()
  MaterialFunctionCall!(
   Input Alpha=LinearInterpolate(A=LinearInterpolate(A=@n2, B=@n3, Alpha=@n4) {constA=0.2, constB=0.4}, B=Constant() {desc=empty image alpha, r=0.2}, Alpha=@n1) {constA=0.2, constB=0.5}
   bOverrideCornerRadius
   Corner Radius Override
  ) {materialFunction=/FortUILibrary/Systems/ItemShopTile/Materials/MaterialFunctions/MF_UI_ItemShop_RoundedCorners}
  n5 = Constant() {r=2}
  n6 = Constant() {r=0.125}
  n7 = MaterialFunctionCall(GradientEnd=MaterialFunctionCall("Value (0-1)"=MaterialFunctionCall!(Animation Duration=@n5, Delay=@n6, Reset to Default) {materialFunction=/FortUILibrary/MaterialFunctions/Time/MF_UI_AnimationDelay}.Result) {materialFunction=&MF_UI_EaseCurves}.EaseInOutSine) {materialFunction=&MF_UI_LinearGradient}
  n8 = MaterialFunctionCall!(A=@n7.Result, B=Constant() {r=1}) {materialFunction=/FortUILibrary/MaterialFunctions/Basics/MF_UI_BooleanOperations}
  NamedRerouteDeclaration(None=Multiply(A=@n7.Result, B=OneMinus(None=@n8.IsEqual))) {name=Loading Mask, nodeColor=<rgba 0 0 0 1>}
  n9 = MaterialFunctionCall() {materialFunction=&GetUserInterfaceUV}
  n10 = NamedRerouteDeclaration(None=TextureSample(
    UVs=MaterialFunctionCall(UV=@n9.Normalized UV) {materialFunction=&MF_UI_UVScaling}.Scaled to Fit
    Tex=TextureObject() {samplerType=SAMPLERTYPE_LinearColor, texture=/FortUILibrary/Textures/Icons/T_UI_Icon_UIKit_ImageError_SDF}
   ) {samplerSource=SSM_Clamp_WorldGroupSettings}.R) {name=Empty Image, nodeColor=<rgba 0 0 0 1>}
  n11 = Constant() {r=0.68}
  NamedRerouteDeclaration(None=LinearInterpolate(Alpha=MaterialFunctionCall(DistanceField=@n10, Threshold=@n11) {materialFunction=&MF_UI_AntiAliasing}.Inner) {constA=0.2, constB=0.5}) {name=Empty Thumbnail, nodeColor=<rgba 0 0 0 1>}
 ]
}
