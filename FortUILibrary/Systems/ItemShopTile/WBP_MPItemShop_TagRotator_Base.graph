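# Base CommonUserWidget for an item shop tile tag that shows an icon-sized badge, expands on focus, and rotates through tag texts.
# OnTextsUpdated stores the texts, toggles enhanced visibility and, for multiple titles in full-preview mode, runs a looping timer with TextOut/TextIn transitions.
# Set Focused State plays hover animations that drive material expansion and lerp spacer widths, also expanding for touch input when the Bacchus frontend is enabled.
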WidgetBlueprint /FortUILibrary/Systems/ItemShopTile/WBP_MPItemShop_TagRotator_Base
animations [
 TextOut {
  binding_count=2
  length_seconds=0.35
  object=&self:TextOut
  sequence {bindings [
    Border_Text {
     id=#1
     possessed_class=Border
     tracks [
      {
       type=MovieScene2DTransformTrack
       sections [{
         type=MovieScene2DTransformSection
         end_frame=7
         start_frame=0
         channels [Translation.X {type=MovieSceneScriptingFloatChannel, keys [[0,0],[7,-100]]}]
        }]
      }
      {
       type=MovieSceneFloatTrack
       display_name=Render Opacity
       sections [{type=MovieSceneFloatSection, end_frame=7, start_frame=0, channels [None {type=MovieSceneScriptingFloatChannel, keys [[0,1],[4,0],[7,0]]}]}]
      }
     ]
    }
    Border_Background {id=#2, possessed_class=Border, tracks [{type=MovieSceneWidgetMaterialTrack, sections []}]}
   ]}
 }
 TextIn {
  binding_count=1
  length_seconds=0.3
  object=&self:TextIn
  sequence {bindings [Border_Text {
     id=#3
     possessed_class=Border
     tracks [
      {
       type=MovieSceneFloatTrack
       display_name=Render Opacity
       sections [{type=MovieSceneFloatSection, end_frame=6, start_frame=0, channels [None {type=MovieSceneScriptingFloatChannel, keys [[0,0],[2,0],[6,1]]}]}]
      }
      {
       type=MovieScene2DTransformTrack
       sections [{
         type=MovieScene2DTransformSection
         end_frame=6
         start_frame=0
         channels [
          Translation.X {type=MovieSceneScriptingFloatChannel, keys [[0,100],[6,0]]}
          Translation.Y {type=MovieSceneScriptingFloatChannel}
         ]
        }]
      }
     ]
    }]}
 }
 HoverFull {
  binding_count=3
  length_seconds=0.15
  object=&self:HoverFull
  sequence {
   bindings [
    WBP_MPItemShop_TagRotator_Base_C_52 {
     id=#4
     possessed_class=WBP_MPItemShop_TagRotator_Base_C
     tracks [{
       type=MovieSceneDoubleTrack
       display_name=Hover Animation Lerp
       sections [{type=MovieSceneDoubleSection, end_frame=9, start_frame=0, channels [None {type=MovieSceneScriptingDoubleChannel, keys [[0,0],[9,1]]}]}]
      }]
    }
    Border_Background {
     id=#5
     possessed_class=Border
     tracks [{
       type=MovieSceneWidgetMaterialTrack
       sections [{
         type=MovieSceneParameterSection
         end_frame=9
         start_frame=0
         channels [Rift energy expand {type=MovieSceneScriptingFloatChannel, keys [[0,0],[9.001,1]]}]
        }]
      }]
    }
    Border_Text {
     id=#6
     possessed_class=Border
     tracks [
      {
       type=MovieSceneFloatTrack
       display_name=Render Opacity
       sections [{type=MovieSceneFloatSection, channels [None {type=MovieSceneScriptingFloatChannel, keys [[0,0],[4,1],[9.001,1]]}]}]
      }
      {
       type=MovieScene2DTransformTrack
       sections [{
         type=MovieScene2DTransformSection
         end_frame=9
         start_frame=0
         channels [Translation.X {type=MovieSceneScriptingFloatChannel, keys [[0,64],[9.001,0]]}]
        }]
      }
     ]
    }
   ]
   display_rate {numerator=60}
  }
 }
 UnhoverFull {
  binding_count=3
  length_seconds=0.15
  object=&self:UnhoverFull
  sequence {
   bindings [
    Border_Background {
     id=#7
     possessed_class=Border
     tracks [{
       type=MovieSceneWidgetMaterialTrack
       sections [{
         type=MovieSceneParameterSection
         end_frame=9
         start_frame=0
         channels [Rift energy expand {type=MovieSceneScriptingFloatChannel, keys [[0,1],[9.001,0]]}]
        }]
      }]
    }
    Border_Text {
     id=#8
     possessed_class=Border
     tracks [{
       type=MovieSceneFloatTrack
       display_name=Render Opacity
       sections [{type=MovieSceneFloatSection, channels [None {type=MovieSceneScriptingFloatChannel, keys [[0,1],[9.001,0]]}]}]
      }]
    }
    WBP_MPItemShop_TagRotator_Base_C_51 {
     id=#9
     possessed_class=WBP_MPItemShop_TagRotator_Base_C
     tracks [{
       type=MovieSceneDoubleTrack
       display_name=Hover Animation Lerp
       sections [{type=MovieSceneDoubleSection, end_frame=9, start_frame=0, channels [None {type=MovieSceneScriptingDoubleChannel, keys [[0,1],[9.001,0]]}]}]
      }]
    }
   ]
   display_rate {numerator=60}
  }
 }
 HoverCollapsed {
  binding_count=2
  length_seconds=0.1507
  object=&self:HoverCollapsed
  sequence {bindings [
    Border_Text {
     id=#10
     possessed_class=Border
     tracks [{
       type=MovieSceneFloatTrack
       display_name=Render Opacity
       sections [{type=MovieSceneFloatSection, end_frame=3, start_frame=0, channels [None {type=MovieSceneScriptingFloatChannel, const [0,2]}]}]
      }]
    }
    Border_Background {
     id=#11
     possessed_class=Border
     tracks [{
       type=MovieScene2DTransformTrack
       sections [{
         type=MovieScene2DTransformSection
         channels [
          Scale.X {type=MovieSceneScriptingFloatChannel, keys [[0,0.85],[3.0133,1]]}
          Scale.Y {type=MovieSceneScriptingFloatChannel, keys [[0,0.85],[3.0133,1]]}
         ]
        }]
      }]
    }
   ]}
 }
 UnhoverCollapsed {
  binding_count=2
  length_seconds=0.1507
  object=&self:UnhoverCollapsed
  sequence {bindings [
    Border_Text {
     id=#12
     possessed_class=Border
     tracks [{
       type=MovieSceneFloatTrack
       display_name=Render Opacity
       sections [{type=MovieSceneFloatSection, end_frame=3, start_frame=0, channels [None {type=MovieSceneScriptingFloatChannel, const [0,2]}]}]
      }]
    }
    Border_Background {
     id=#13
     possessed_class=Border
     tracks [{
       type=MovieScene2DTransformTrack
       sections [{
         type=MovieScene2DTransformSection
         channels [
          Scale.X {type=MovieSceneScriptingFloatChannel, keys [[0,1],[3.0147,0.9]]}
          Scale.Y {type=MovieSceneScriptingFloatChannel, keys [[0,1],[3.0147,0.9]]}
         ]
        }]
      }]
    }
   ]}
 }
]
graphs [
 EventGraph {nodes [
   n1 = Event PreConstruct()
   > InitializeTagRotator()
   > n2 = Branch(Condition=@n1.IsDesignTime)
   > n3 = Set IsFullPreviewEnabled(IsFullPreviewEnabled:bool=Get PreviewFullInfoEnabled() -> (PreviewFullInfoEnabled:bool)) -> (Output_Get:bool)
   > Set Focused State!(execute, self, IsFocused:bool=@n3, IsImmediate:bool=true) -> (then) {is=title K2Node_CallFunction}
   > Set Text(Text=Get PreviewTitleText() -> (PreviewTitleText:Text), self=Get Text_Tag() -> (Text_Tag:Common Text Object Reference))
   n4 = Get PreviewTitleText()
   > OnEnhancedUpdated!(execute, self, IsEnhanced:bool=NOT Boolean(A=TextIsEmpty(InText=@n4))) -> (then)
   Set Focused State(execute=@n2.else, IsFocused=false, IsImmediate=true)
   Event Destruct()
   > Clear and Invalidate Timer by Handle(Handle=Get TimerHandle() -> (TimerHandle:Timer Handle Structure))
   K2Node_Event_3 = OnTagEnter() {is=title K2Node_CustomEvent}
   > "Play Animation with Finished event (legacy)"!(
    execute:Exec
    Widget:User Widget Object Reference=Self-Reference()
    InAnimation:Widget Animation Object Reference=Get TextIn() -> (TextIn:Widget Animation Object Reference)
    StartAtTime:float=0
    NumLoopsToPlay:int=1
    PlayMode:EUMGSequencePlayMode Enum
    PlaybackSpeed:float=1
   ) -> (then:Exec, Result:UMGSequence Player Object Reference, Finished:Exec) {is=title K2Node_PlayAnimation}
   K2Node_Event_4 = OnTagExit() {is=title K2Node_CustomEvent}
   > n5 = "Play Animation with Finished event (legacy)"(
    Widget=Self-Reference()
    InAnimation=Get TextOut() -> (TextOut:Widget Animation Object Reference)
    StartAtTime=0
    NumLoopsToPlay=1
    PlaybackSpeed=1
   )
   UpdateTagText(execute=@n5.Finished)
   > Branch(Condition=Get IsFocused() -> (IsFocused:bool))
   > OnTagEnter() {is=title K2Node_CallFunction}
  ]}
 InitializeTagRotator {nodes [
   InitializeTagRotator() {is=title K2Node_FunctionEntry}
   > n1 = Sequence!(execute) -> (then_0, then_1, then_2:Exec, then_3:Exec)
   n2 = Get HeightOverride!(self:Size Box Object Reference=Get SizeBox_All() -> (SizeBox_All:Size Box Object Reference)) -> (HeightOverride:float)
   n3 = Set!(
    execute=@n1.then_1
    Size_X:double=float / float(A=@n2, B=1.5)
    Size_Y:double=@n2
    self:Spacer Object Reference=Get Spacer_IconSpace_Right() -> (Spacer_IconSpace_Right:Spacer Object Reference)
   ) -> (then, Output_Get:Vector 2D Structure)
   > Set IconSpacerSize(IconSpacerSize:Vector 2D Structure=@n3) -> (Output_Get:Vector 2D Structure)
   > Set HeightOverride!(
    execute
    HeightOverride:float=SelectFloat(A=Get TagHeight_Mobile() -> (TagHeight_Mobile:double), B=Get TagHeight() -> (TagHeight:double), bPickA=IsMobileGame())
    self:Size Box Object Reference=Get SizeBox_All()
   ) -> (then, Output_Get:float)
   SetVisibility(
    execute=@n1.then_2
    self=Get Border_Text() -> (Border_Text:Border Object Reference)
    InVisibility=Select(Option 0=Collapsed, Option 1=SelfHitTestInvisible, Index=Get IsFullPreviewEnabled() -> (IsFullPreviewEnabled:bool))
   )
   > Set Font!(execute, Font:Slate Font Info Structure=Get TagFont() -> (TagFont:Slate Font Info Structure), self:Text Object Reference=Get Text_Tag()) -> (then, Output_Get:Slate Font Info Structure)
   > Set TextTransformPolicy!(
    execute
    TextTransformPolicy:ETextTransformPolicy Enum=Get TagTransformPolicy() -> (TagTransformPolicy:ETextTransformPolicy Enum=None)
    self:Text Object Reference=Get Text_Tag()
   ) -> (then, Output_Get:ETextTransformPolicy Enum=None)
   > Set Padding!(
    execute
    Padding:Margin Structure=Get TagPadding() -> (TagPadding:Margin Structure)
    self:Size Box Slot Object Reference=SlotAsSizeBoxSlot!(Widget:Widget Object Reference=Get Border_Background() -> (Border_Background:Border Object Reference)) -> (ReturnValue:Size Box Slot Object Reference)
   ) -> (then, Output_Get:Margin Structure)
   > SetColorAndOpacity(self=Get Text_Tag(), InColorAndOpacity=Get TextColor() -> (TextColor:Slate Color Structure))
   n4 = Get Border_Background()
   SetBrushFromMaterial(execute=@n1.then_0, self:Border Object Reference=@n4, Material=Get TagMaterial() -> (TagMaterial:Material Interface Object Reference))
   > n5 = GetDynamicMaterial(self:Border Object Reference=@n4)
   > Set TagMID(TagMID:Material Instance Dynamic Object Reference=@n5) -> (Output_Get:Material Instance Dynamic Object Reference)
  ]}
 Is TagText Visible {nodes [
   Is TagText Visible() {is=title K2Node_FunctionEntry}
   n1 = "Nearly Equal (Float)"(A=Get RenderOpacity!(self:Widget Object Reference=Get Border_Text()) -> (RenderOpacity:float), B=0, ErrorTolerance=0)
   n2 = "Equal (Integer)"(A=Length(TargetArray:Array of Texts=Get TagTexts() -> (TagTexts:Array of Texts)))
   > Return Node!(execute, bTagIsVisible:bool=NOT Boolean(A=OR Boolean(A=@n2, B=@n1)))
  ]}
 OnEnhancedUpdated {nodes [
   n1 = OnEnhancedUpdated!() -> (then, IsEnhanced:bool=~) {is=title K2Node_FunctionEntry}
   > n2 = Set IsEnhanced(IsEnhanced:bool=@n1) -> (Output_Get:bool)
   > SetScalarParameterValue(self=Get TagMID() -> (TagMID:Material Instance Dynamic Object Reference), ParameterName=bIsEnhanced, Value="To Float (Boolean)"(InBool=@n2))
   > Set Visibility(Visibility=Select(Option 0=Collapsed, Option 1=SelfHitTestInvisible, Index=@n1), self=Get SizeBox_All())
  ]}
 OnTextsUpdated {nodes [
   OnTextsUpdated!() -> (then, Texts:Array of Texts) {is=title K2Node_FunctionEntry}
   > Clear and Invalidate Timer by Handle(Handle=Get TimerHandle())
   > n1 = Set TagTexts(TagTexts:Array of Texts=Get Texts() -> (Texts:Array of Texts)) -> (Output_Get:Array of Texts)
   > Set NumTitles(NumTitles:int=Length(TargetArray=@n1)) -> (Output_Get:int=0)
   > n2 = Branch(Condition=Get IsFullPreviewEnabled())
   > n3 = Switch on Int!(execute:Exec, Selection:int=Get NumTitles() -> (NumTitles:int=0)) -> (Default:Exec, 0:Exec, 1:Exec)
   > OnEnhancedUpdated(IsEnhanced=true) {is=title K2Node_CallFunction}
   > n4 = Branch(Condition=Get IsFocused())
   > ShowFirstTitle()
   > n5 = Set Timer by Event(Delegate=Create Event(), Time=Get TimeBetweenTitleChanges() -> (TimeBetweenTitleChanges:float), bLooping=true)
   > Set TimerHandle(TimerHandle:Timer Handle Structure=@n5) -> (Output_Get:Timer Handle Structure)
   UpdateTagText(execute=@n4.else)
   OnEnhancedUpdated(execute=@n3.0, IsEnhanced=false)
   OnEnhancedUpdated(execute=@n3.1, IsEnhanced=true)
   > n6 = Branch(Condition=Get IsFocused())
   > ShowFirstTitle()
   UpdateTagText(execute=@n6.else)
   OnEnhancedUpdated(execute=@n2.else, IsEnhanced="integer > integer"(A=Get NumTitles() -> (NumTitles=0)))
  ]}
 PlayHoverAnimation {nodes [
   n1 = PlayHoverAnimation!() -> (then, IsImmediate:bool=~) {is=title K2Node_FunctionEntry}
   > n2 = Branch(Condition=@n1)
   > n3 = Set RenderOpacity!(execute, RenderOpacity:float=1, self:Widget Object Reference=Get Border_Text()) -> (then, Output_Get:float)
   > SetHoverAnimationLerp!(execute, self, LerpValue:double=@n3) -> (then)
   > SetScalarParameterValue(self=Get TagMID(), ParameterName=Rift energy expand, Value=1)
   PlayAnimationForward(
    execute=@n2.else
    InAnimation=Select(
     Option 0:Widget Animation Object Reference=Get HoverCollapsed() -> (HoverCollapsed:Widget Animation Object Reference)
     Option 1:Widget Animation Object Reference=Get HoverFull() -> (HoverFull:Widget Animation Object Reference)
     Index=Get IsFullPreviewEnabled()
    ) -> (ReturnValue:Widget Animation Object Reference)
   )
  ]}
 PlayUnhoverAnimation {nodes [
   n1 = PlayUnhoverAnimation!() -> (then, IsImmediate:bool=~) {is=title K2Node_FunctionEntry}
   > n2 = Branch(Condition=@n1)
   > n3 = Set RenderOpacity(RenderOpacity=0, self=Get Border_Text())
   > SetHoverAnimationLerp(LerpValue=@n3)
   > SetScalarParameterValue(self=Get TagMID(), ParameterName=Rift energy expand, Value=0)
   PlayAnimationForward(
    execute=@n2.else
    InAnimation=Select(
     Option 0=Get UnhoverCollapsed() -> (UnhoverCollapsed:Widget Animation Object Reference)
     Option 1=Get UnhoverFull() -> (UnhoverFull:Widget Animation Object Reference)
     Index=Get IsFullPreviewEnabled()
    )
   )
  ]}
 Set Focused State {nodes [
   n1 = Set Focused State!() -> (then, IsFocused:bool=~, IsImmediate:bool=~) {is=title K2Node_FunctionEntry}
   > Set IsFocused(IsFocused:bool=@n1) -> (Output_Get:bool)
   n2 = AND Boolean(
    A=IsUsingTouch(PlayerController=GetOwningPlayer())
    B=Get bBacchusFrontendEnabled!(self:Fort Runtime Options Object Reference=GetRuntimeOptions()) -> (bBacchusFrontendEnabled:bool)
   )
   n3 = Get IsFocused()
   > n4 = Branch(Condition=AND Boolean(A=OR Boolean(A=@n3, B=@n2), B=Get IsFullPreviewEnabled()))
   n5 = Get Border_Text()
   > SetVisibility(self=@n5, InVisibility=SelfHitTestInvisible)
   > Set CurrentTitleIndex(CurrentTitleIndex:int=0) -> (Output_Get:int=0)
   > PlayHoverAnimation!(execute, self, IsImmediate:bool=Get IsImmediate() -> (IsImmediate:bool)) -> (then) {is=title K2Node_CallFunction}
   > OnTextsUpdated!(execute, self, Texts:Array of Texts=Get TagTexts()) -> (then) {is=title K2Node_CallFunction}
   SetVisibility(execute=@n4.else, self=@n5, InVisibility=Collapsed)
   > Clear and Invalidate Timer by Handle(Handle=Get TimerHandle())
   > Set CurrentTitleIndex(CurrentTitleIndex=0) -> (Output_Get=0)
   > OnEnhancedUpdated(IsEnhanced=Get IsEnhanced() -> (IsEnhanced:bool))
   > PlayUnhoverAnimation!(execute, self, IsImmediate:bool=Get IsImmediate()) -> (then) {is=title K2Node_CallFunction}
  ]}
 SetHoverAnimationLerp {nodes [
   SetHoverAnimationLerp!() -> (then, LerpValue:double=~) {is=title K2Node_FunctionEntry}
   > Set(
    Size_X=Lerp(A=Get!() -> (IconSpacerSize_X:double, IconSpacerSize_Y:double), B=0, Alpha=Get LerpValue() -> (LerpValue:double))
    Size_Y=Get().IconSpacerSize_Y
    self=Get Spacer_IconSpace_Right()
   )
   > Set(
    Size_X=Lerp(A=0, B=Get(), Alpha=Get LerpValue())
    Size_Y=Get().IconSpacerSize_Y
    self=Get Spacer_IconSpace_Left() -> (Spacer_IconSpace_Left:Spacer Object Reference)
   )
  ]}
 ShowFirstTitle {nodes [
   ShowFirstTitle() {is=title K2Node_FunctionEntry}
   > UpdateTagText()
   > OnTagEnter()
  ]}
 TransitionToNextTitle {nodes [
   TransitionToNextTitle() {is=title K2Node_FunctionEntry}
   n1 = Get NumTitles() -> (NumTitles=0)
   n2 = Get CurrentTitleIndex() -> (CurrentTitleIndex:int=0)
   > Set CurrentTitleIndex(CurrentTitleIndex="Wrap (Integer)"!(Value:int=int + int(A=@n2), Min:int=-1, Max:int=int - int(A=@n1)) -> (ReturnValue:int=0)) -> (Output_Get=0)
   > OnTagExit() {is=title K2Node_CallFunction}
  ]}
 UpdateTagText {nodes [
   UpdateTagText() {is=title K2Node_FunctionEntry}
   > n1 = Set TitleText(TitleText:Text="Get (a ref)"(Array:Array of Texts=Get TagTexts(), Dimension 1=Get CurrentTitleIndex() -> (CurrentTitleIndex=0)) -> (Output:"Text (by ref)")) -> (Output_Get:Text)
   > "SetText (Text)"(self=Get Text_Tag(), InText=@n1)
  ]}
]
variables [
 CurrentTitleIndex {type=integer}
 HoverAnimationLerp {type=number}
 IconSpacerSize {type=Vector2D, default=<0 0>}
 IsEnhanced {type=boolean}
 IsFocused {type=boolean}
 IsFullPreviewEnabled {type=boolean}
 NumTitles {type=integer}
 PreviewFullInfoEnabled {type=boolean, default=true}
 PreviewTitleText {type=string, default="[PH] Selectable styles"}
 TagFont {
  type=SlateFontInfo
  default {
   bForceMonospaced=false
   bMaterialIsStencil=false
   fontMaterial=null
   fontObject=&HeadingNow
   letterSpacing=0
   monospacedWidth=1
   size=13.5
   skewAmount=0
   typefaceFontName=75 Medium Italic
   outlineSettings {
    bApplyOutlineToDropShadows=false
    bMiteredCorners=false
    bSeparateFillAlpha=false
    outlineMaterial=null
    outlineSize=2
    outlineColor=<rgba 0 0 0 0>
   }
  }
 }
 TagHeight {type=number, default=40}
 TagHeight_Mobile {type=number, default=50}
 TagMaterial {type=MaterialInterface, default=/FortUILibrary/Systems/Items/MaterialInstances/MI_UIKit_Transference}
 TagMID {type=MaterialInstanceDynamic}
 TagPadding {type=Margin, default=<ltrb 0 0 0 0>}
 TagTexts {type="array<string>", default []}
 TagTransformPolicy {type=ETextTransformPolicy, default=None}
 TextColor {type=SlateColor, default {colorUseRule=UseColor_Specified, specifiedColor=<rgba 1 1 1 1>}}
 TimeBetweenTitleChanges {type=number, default=2}
 TimerHandle {type=TimerHandle}
]
widgets {
 info {inheritedWidgetCount=0, parentClass=&CommonUserWidget, rootWidgetClass=&SizeBox, widgetCount=7}
 widgets [
  SizeBox_All {widgetClassPath=&SizeBox, bIsVariable=true, widget=SizeBox_All, properties {bOverride_HeightOverride=true, heightOverride=48}}
  Border_Background {
   widgetClassPath=&Border
   bIsVariable=true
   parent=SizeBox_All
   widget=Border_Background
   properties {
    background {
     bIsDynamicallyLoaded=false
     drawAs=Image
     imageType=NoImage
     mirroring=NoMirror
     resourceName=None
     resourceObject=&MI_UIKit_Transference
     tiling=NoTile
     imageSize=<32 32>
     margin=<ltrb 0 0 0 0>
     outlineSettings {
      bUseBrushTransparency=false
      roundingType=HalfHeightRadius
      width=0
      color {colorUseRule=UseColor_Specified, specifiedColor=<rgba 0 0 0 0>}
      cornerRadii=<0 0 0 0>
     }
     tintColor {colorUseRule=UseColor_Specified, specifiedColor=<rgba 1 1 1 1>}
     uVRegion {bIsValid=false, max=<0 0>, min=<0 0>}
    }
    padding=<ltrb 0 0 0 0>
    renderTransformPivot=<1 1>
   }
   slot {class=SizeBoxSlot}
  }
  HorizontalBox_0 {
   widgetClassPath=&HorizontalBox
   parent=Border_Background
   widget=HorizontalBox_0
   properties {clipping=ClipToBounds}
   slot {class=BorderSlot, properties {padding=<ltrb 0 0 0 0>}}
  }
  Spacer_IconSpace_Left {
   widgetClassPath=&Spacer
   bIsVariable=true
   parent=HorizontalBox_0
   widget=Spacer_IconSpace_Left
   properties {size=<0 48>}
   slot {class=HorizontalBoxSlot, properties {horizontalAlignment=HAlign_Right, verticalAlignment=VAlign_Bottom}}
  }
  Border_Text {
   widgetClassPath=&Border
   bIsVariable=true
   parent=HorizontalBox_0
   widget=Border_Text
   properties {brushColor=<rgba 1 1 1 0>, padding=<ltrb 0 0 0 0>}
   slot {class=HorizontalBoxSlot, properties {verticalAlignment=VAlign_Center, size {sizeRule=Fill, value=1}}}
  }
  Text_Tag {
   widgetClassPath=&CommonTextBlock
   bIsVariable=true
   parent=Border_Text
   widget=Text_Tag
   properties {
    mobileFontSizeMultiplier=1.34
    scrollStyle=&TextScrollStyle-Base_C
    text="[PH] Selectable styles"
    textTransformPolicy=ToUpper
    font {
     bForceMonospaced=false
     bMaterialIsStencil=false
     fontMaterial=null
     fontObject=&BurbankSmall
     letterSpacing=0
     monospacedWidth=1
     size=12
     skewAmount=0.2
     typefaceFontName=Black
     outlineSettings {
      bApplyOutlineToDropShadows=false
      bMiteredCorners=false
      bSeparateFillAlpha=false
      outlineMaterial=null
      outlineSize=2
      outlineColor=<rgba 0 0 0 0>
     }
    }
    renderTransform {angle=0, scale=<1 1>, shear=<0 0>, translation=<0 2>}
    shadowOffset=<0 0>
   }
   slot {class=BorderSlot, properties {padding=<ltrb 0 0 0 0>}}
  }
  Spacer_IconSpace_Right {
   widgetClassPath=&Spacer
   bIsVariable=true
   parent=HorizontalBox_0
   widget=Spacer_IconSpace_Right
   properties {size=<48 48>}
   slot {class=HorizontalBoxSlot, properties {horizontalAlignment=HAlign_Right, verticalAlignment=VAlign_Bottom}}
  }
 ]
}
