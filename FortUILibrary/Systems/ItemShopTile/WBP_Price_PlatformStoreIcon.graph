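# Item shop price widget that shows a PlayStation Store logo beside real-money prices.
# PreConstruct applies the configurable image size, then refreshes the logo's visibility.
# The logo shows only when the currency is RealMoney and the platform is PS4 or PS5, or a preview flag is set in editor or PIE, otherwise it collapses.
# UpdateCurrency stores a new currency type and re-runs the visibility check.
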WidgetBlueprint /FortUILibrary/Systems/ItemShopTile/WBP_Price_PlatformStoreIcon
graphs [
 EventGraph {nodes [
   Event PreConstruct()
   > SetDesiredSizeOverride!(
    execute
    self:Image Object Reference=Get PlayStationStoreLogo() -> (PlayStationStoreLogo:Image Object Reference)
    DesiredSize:Vector 2D Structure=Get ImageSize() -> (ImageSize:Vector 2D Structure)
   ) -> (then)
   > RefreshIconVisibility()
  ]}
 RefreshIconVisibility {nodes [
   RefreshIconVisibility() {is=title K2Node_FunctionEntry}
   n1 = GetPlatformName!() -> (ReturnValue:String)
   n2 = OR Boolean!(
    A="Equal, Case Insensitive (String)"(A=@n1, B=PS4)
    B="Equal, Case Insensitive (String)"(A=@n1, B=PS5)
    C:bool=AND Boolean(A=Get PreviewPlayStationPlatform() -> (PreviewPlayStationPlatform:bool), B=InEditorOrPIE!() -> (ReturnValue:bool))
   ) -> (ReturnValue)
   n3 = "Equal (Enum)"(A:EFortPriceCurrency Enum=Get Currency() -> (Currency:EFortPriceCurrency Enum=Mtx), B:EFortPriceCurrency Enum=RealMoney)
   > Set Visibility(Visibility=Select(Option 0=Collapsed, Option 1=SelfHitTestInvisible, Index=AND Boolean(A=@n2, B=@n3)), self=Get PlayStationStoreLogo())
  ]}
 UpdateCurrency {nodes [
   n1 = UpdateCurrency!() -> (then, CurrencyType:EFortPriceCurrency Enum=Mtx) {is=title K2Node_FunctionEntry}
   > Set Currency(Currency:EFortPriceCurrency Enum=@n1) -> (Output_Get:EFortPriceCurrency Enum=Mtx)
   > RefreshIconVisibility() {is=title K2Node_CallFunction}
  ]}
]
variables [
 Currency {type=EFortPriceCurrency, default=RealMoney}
 ImageSize {type=Vector2D, default=<32 32>}
 PreviewPlayStationPlatform {type=boolean}
]
widgets {
 info {inheritedWidgetCount=0, parentClass=&UserWidget, rootWidgetClass=&Image, widgetCount=1}
 widgets [PlayStationStoreLogo {
   widgetClassPath=&Image
   bIsVariable=true
   widget=PlayStationStoreLogo
   properties {
    visibility=SelfHitTestInvisible
    brush {
     bIsDynamicallyLoaded=false
     drawAs=Image
     imageType=NoImage
     mirroring=NoMirror
     resourceName=None
     resourceObject=/Game/UI/Frontend/ItemShop/Art/T_UI_PS_Store_Logo_64x
     tiling=NoTile
     imageSize=<32 32>
     margin=<ltrb 0 0 0 0>
     outlineSettings {
      bUseBrushTransparency=false
      roundingType=HalfHeightRadius
      width=0
      color {colorUseRule=UseColor_Specified, specifiedColor=<rgba 0 0 0 0>}
      cornerRadii=<0 0 0 0>
     }
     tintColor {colorUseRule=UseColor_Specified, specifiedColor=<rgba 1 1 1 1>}
     uVRegion {bIsValid=false, max=<0 0>, min=<0 0>}
    }
    renderTransformPivot=<1 1>
   }
  }]
}
