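# Base Item Shop tile widget that switches between price, owned and subscription status displays.
# GetPriceInfoState derives that state from display mode, ownership and offer type; setters fill currency icon, price, struck-through regular price and optional real-price line, which adjusts height.
# PreConstruct applies exposed styling and design-time previews, while SetMobileState adjusts padding and icon scale.
# Colors are written to dynamic material parameters, and hover animates the materials' Trans_Focused parameter.
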WidgetBlueprint /FortUILibrary/Systems/ItemShopTile/WBP_PriceSwitcher_Base
animations [Hover {
  binding_count=3
  length_seconds=0.1
  object=&self:Hover
  sequence {bindings [
    Text_Price {
     id=#1
     possessed_class=CommonTextBlock
     tracks [{
       type=MovieSceneWidgetMaterialTrack
       display_name=Font.FontMaterial
       sections [{
         type=MovieSceneParameterSection
         end_frame=2
         start_frame=0
         channels [Trans_Focused {type=MovieSceneScriptingFloatChannel, keys [[0,0],[1.9997,1]]}]
        }]
      }]
    }
    Text_RegularPrice {
     id=#2
     possessed_class=CommonTextBlock
     tracks [{
       type=MovieSceneWidgetMaterialTrack
       display_name=Font.FontMaterial
       sections [{
         type=MovieSceneParameterSection
         end_frame=2
         start_frame=0
         channels [Trans_Focused {type=MovieSceneScriptingFloatChannel, keys [[0,0],[1.9997,1]]}]
        }]
      }]
    }
    Image_Strike {
     id=#3
     possessed_class=Image
     tracks [{
       type=MovieSceneWidgetMaterialTrack
       sections [{
         type=MovieSceneParameterSection
         end_frame=2
         start_frame=0
         channels [Trans_Focused {type=MovieSceneScriptingFloatChannel, keys [[0,0],[1.9997,1]]}]
        }]
      }]
    }
   ]}
 }]
graphs [
 AlphaOverride {nodes [
   n1 = Inputs!() -> (LinearColor:"Linear Color Structure (by ref)", NewAlpha:float=~)
   n2 = Break Linear Color!(LinearColor:"Linear Color Structure (by ref)"=@n1) -> (R:float=~, G:float=~, B:float=~, A:float=~) {is=title K2Node_BreakStruct}
   Outputs!(ReturnValue:Linear Color Structure=MakeColor!(R:float=@n2, G:float=@n2.G, B:float=@n2.B, A:float=@n1.NewAlpha) -> (ReturnValue:Linear Color Structure))
  ]}
 EventGraph {nodes [
   OnUnhovered() {is=title K2Node_CustomEvent}
   n1 = Get Hover() -> (Hover:Widget Animation Object Reference)
   > PlayAnimationReverse(InAnimation=@n1, PlaybackSpeed=2)
   OnHovered() {is=title K2Node_CustomEvent}
   > PlayAnimationForward(InAnimation=@n1)
   n2 = SetMobileState!() -> (OutputDelegate, then, InMobileState:bool=~) {is=title K2Node_CustomEvent}
   > SetRenderScale!(
    execute
    self:Widget Object Reference=Get Image_Currency() -> (Image_Currency:Image Object Reference)
    Scale:Vector 2D Structure=Select(
     Option 0:Vector 2D Structure=Get DefaultCurrencyImageScale() -> (DefaultCurrencyImageScale:Vector 2D Structure)
     Option 1:Vector 2D Structure=Get MobileCurrencyImageScale() -> (MobileCurrencyImageScale:Vector 2D Structure)
     Index=@n2.InMobileState
    ) -> (ReturnValue:Vector 2D Structure)
   ) -> (then)
   > SetPadding!(
    execute
    self:Horizontal Box Slot Object Reference=SlotAsHorizontalBoxSlot!(Widget:Widget Object Reference=Get Text_Price() -> (Text_Price:Common Text Object Reference)) -> (ReturnValue:Horizontal Box Slot Object Reference)
    InPadding:Margin Structure=Select(
     Option 0:Margin Structure=Get DefaultPricePadding() -> (DefaultPricePadding:Margin Structure)
     Option 1:Margin Structure=Get MobilePricePadding() -> (MobilePricePadding:Margin Structure)
     Index=@n2.InMobileState
    ) -> (ReturnValue:Margin Structure)
   ) -> (then)
   > SetPadding(
    self:Overlay Slot Object Reference=SlotAsOverlaySlot(Widget=Get RegularHorizontal() -> (RegularHorizontal:Horizontal Box Object Reference))
    InPadding=Select(
     Option 0=Get DefaultRegularPricePadding() -> (DefaultRegularPricePadding:Margin Structure)
     Option 1=Get MobileRegularPricePadding() -> (MobileRegularPricePadding:Margin Structure)
     Index=@n2.InMobileState
    )
   )
   > SetPadding(
    self:Horizontal Box Slot Object Reference=SlotAsHorizontalBoxSlot(Widget=Get Text_AdditionalPriceText() -> (Text_AdditionalPriceText:Common Text Object Reference))
    InPadding=Select(
     Option 0=Get AdditionalPriceTextPadding() -> (AdditionalPriceTextPadding:Margin Structure)
     Option 1=Get AdditionalPriceTextMobilePadding() -> (AdditionalPriceTextMobilePadding:Margin Structure)
     Index=@n2.InMobileState
    )
   )
   > SetPadding(
    self:Grid Slot Object Reference=SlotAsGridSlot(Widget=Get Text_RealPrice() -> (Text_RealPrice:Common Text Object Reference))
    InPadding=Select(
     Option 0=Get DefaultRealPricePadding() -> (DefaultRealPricePadding:Margin Structure)
     Option 1=Get MobileRealPricePadding() -> (MobileRealPricePadding:Margin Structure)
     Index=@n2.InMobileState
    )
   )
   n3 = Event PreConstruct()
   > InitializeWidget()
   > Set PreviewPlayStationPlatform!(
    execute
    PreviewPlayStationPlatform:bool=Get PreviewPlayStationPlatform() -> (PreviewPlayStationPlatform:bool)
    self:WBP Price Platform Store Icon Object Reference=Get WBP_Price_PlatformStoreIcon() -> (WBP_Price_PlatformStoreIcon:WBP Price Platform Store Icon Object Reference)
   ) -> (then, Output_Get:bool)
   > Branch(Condition=@n3.IsDesignTime)
   > SetSwitcherState!(execute, self, PriceState:E_PriceSwitcherState Enum=Get PreviewPriceState() -> (PreviewPriceState:E_PriceSwitcherState Enum=NewEnumerator0)) -> (then)
   > SetPrice!(execute, self, Price:Text="To Text (Integer)"(Value=Get PreviewPrice() -> (PreviewPrice:int=0))) -> (then)
   > SetRegularPrice!(execute, self, Price:Text="To Text (Integer)"(Value=Get PreviewRegularPrice() -> (PreviewRegularPrice:int=0))) -> (then)
   > SetStrikethroughPriceVisibility!(
    execute
    self
    Visibility:ESlateVisibility Enum=GetStrikethroughPriceVisibility!(
     self
     Price:int=Get PreviewPrice() -> (PreviewPrice=0)
     OriginalPrice:int=Get PreviewRegularPrice() -> (PreviewRegularPrice=0)
     SuppressStrikethrough:bool=false
    ) -> (StrikethroughPriceVisibility:ESlateVisibility Enum=Visible)
   ) -> (then)
   > SetCurrency!(execute, self, Currency:EFortPriceCurrency Enum=Get PreviewCurrency() -> (PreviewCurrency:EFortPriceCurrency Enum=Mtx)) -> (then)
   > SetMobileState!(execute, self, InMobileState:bool=Get PreviewMobileState() -> (PreviewMobileState:bool)) -> (then) {is=title K2Node_CallFunction}
   > SetAdditionalPriceText!(execute, self, Text:Text=Get PreviewAdditionalPriceText() -> (PreviewAdditionalPriceText:Text)) -> (then)
   > SetAdditionalRegularPriceText!(execute, self, Text:Text=Get PreviewAdditionalRegularPriceText() -> (PreviewAdditionalRegularPriceText:Text)) -> (then)
   > SetRealPriceText!(
    execute
    self
    Text:Text=Select(Option 0:Text, Option 1:Text=Get PreviewRealPriceText() -> (PreviewRealPriceText:Text), Index=Get HasRealPrice() -> (HasRealPrice:bool)) -> (ReturnValue:Text)
   ) -> (then)
   > SetRegularPriceImageVisibility!(
    execute
    self
    InVisibility:ESlateVisibility Enum=Get CurrencyRegularVisibility() -> (CurrencyRegularVisibility:ESlateVisibility Enum=Visible)
   ) -> (then)
   Event Construct()
   > SetMobileState(InMobileState=IsMobileGame())
   Get Text_RegularPrice() -> (Text_RegularPrice:Common Text Object Reference)
  ]}
 FlipHorizontalBoxLayout {nodes [
   FlipHorizontalBoxLayout() {is=title K2Node_FunctionEntry}
   > n1 = Set HBChildren(HBChildren:Array of Widget Object References=GetAllChildren(self=Get HB_PriceContainer() -> (HB_PriceContainer:Horizontal Box Object Reference))) -> (Output_Get:Array of Widget Object References)
   > Reverse!(execute, TargetArray:Array of Widget Object References=@n1) -> (then) {is=title K2Node_CallArrayFunction}
   > ClearChildren(self=Get HB_PriceContainer())
   > n2 = For Each Loop(Array:Array of Widget Object References=Get HBChildren() -> (HBChildren:Array of Widget Object References)) -> (Array Element:Widget Object Reference)
   > AddChildToHorizontalBox!(execute, self:Horizontal Box Object Reference=Get HB_PriceContainer(), Content:Widget Object Reference=@n2) -> (then, ReturnValue:Horizontal Box Slot Object Reference)
  ]}
 GetPriceInfoState {nodes [
   n1 = GetPriceInfoState!() -> (then, IsOwned:bool=~, OfferType:EFortItemShopOfferType Enum=LevelBundleOffer) {is=title K2Node_FunctionEntry}
   n2 = Select(Option 0:E_PriceSwitcherState Enum=NewEnumerator0, Option 1:E_PriceSwitcherState Enum=NewEnumerator1, Index=@n1) -> (ReturnValue:E_PriceSwitcherState Enum)
   n3 = Get OfferType() -> (OfferType:EFortItemShopOfferType Enum=None)
   n4 = Get OfferType() -> (OfferType=None)
   > Return Node!(
    execute
    SwitcherState:E_PriceSwitcherState Enum=Select!(
     NewEnumerator2:E_PriceSwitcherState Enum=Select!(
      None:E_PriceSwitcherState Enum=@n2
      LevelBundle:E_PriceSwitcherState Enum=@n2
      BattlePass:E_PriceSwitcherState Enum=@n2
      ProgressiveCosmetic:E_PriceSwitcherState Enum=NewEnumerator2
      Crew:E_PriceSwitcherState Enum=NewEnumerator2
      Special:E_PriceSwitcherState Enum=@n2
      Regular:E_PriceSwitcherState Enum=@n2
      MainItemType:E_PriceSwitcherState Enum=NewEnumerator0
      MainItemTypeAndBonus:E_PriceSwitcherState Enum=NewEnumerator0
      Pack:E_PriceSwitcherState Enum=NewEnumerator0
      OutfitPack:E_PriceSwitcherState Enum=NewEnumerator0
      PackAndBonus:E_PriceSwitcherState Enum=NewEnumerator0
      Vehicle:E_PriceSwitcherState Enum=NewEnumerator0
      VehicleAndBonus:E_PriceSwitcherState Enum=NewEnumerator0
      Bundle:E_PriceSwitcherState Enum=@n2
      LEGOKit:E_PriceSwitcherState Enum=NewEnumerator0
      Index:EFortItemShopOfferType Enum=@n3
     ) -> (ReturnValue)
     NewEnumerator0:E_PriceSwitcherState Enum=Select(
      None=NewEnumerator0
      LevelBundle=NewEnumerator0
      BattlePass=NewEnumerator0
      ProgressiveCosmetic=NewEnumerator2
      Crew=NewEnumerator2
      Special=NewEnumerator0
      Regular=NewEnumerator0
      MainItemType=NewEnumerator0
      MainItemTypeAndBonus=NewEnumerator0
      Pack=NewEnumerator0
      OutfitPack=NewEnumerator0
      PackAndBonus=NewEnumerator0
      Vehicle=NewEnumerator0
      VehicleAndBonus=NewEnumerator0
      Bundle=NewEnumerator0
      LEGOKit=NewEnumerator0
      Index=@n4
     )
     NewEnumerator1:E_PriceSwitcherState Enum=NewEnumerator1
     Index:E_PriceDisplayMode Enum=Get PriceDisplayMode() -> (PriceDisplayMode:E_PriceDisplayMode Enum=NewEnumerator2)
    ) -> (ReturnValue)
   )
  ]}
 GetStrikethroughPriceVisibility {nodes [
   n1 = GetStrikethroughPriceVisibility!() -> (then, Price:int, OriginalPrice:int, SuppressStrikethrough:bool=~) {is=title K2Node_FunctionEntry}
   n2 = NOT Boolean(A=@n1.SuppressStrikethrough)
   n3 = "Not Equal (Integer)"(A=@n1, B=@n1.OriginalPrice)
   > Return Node!(
    execute
    StrikethroughPriceVisibility:ESlateVisibility Enum=Select!(Option 0:ESlateVisibility Enum=Collapsed, Option 1:ESlateVisibility Enum=SelfHitTestInvisible, Index:bool=AND Boolean(A=@n3, B=@n2)) -> (ReturnValue:ESlateVisibility Enum)
   )
  ]}
 GetSubInfoColor {nodes [
   n1 = GetSubInfoColor!() -> (then, SubscriptionState:EMcpSubscriptionState Enum=Inactive) {is=title K2Node_FunctionEntry}
   > n2 = GetSubInfoConfig!(execute, self, SubscriptionState:EMcpSubscriptionState Enum=@n1) -> (then, SubInfoText:Text, SubInfoColor:Slate Color Structure)
   > Return Node!(execute, SubInfoColor:Slate Color Structure=@n2.SubInfoColor)
  ]}
 GetSubInfoConfig {nodes [
   GetSubInfoConfig!() -> (then, SubscriptionState:EMcpSubscriptionState Enum=Inactive) {is=title K2Node_FunctionEntry}
   n1 = Get SubscriptionDefaultStatusColor() -> (SubscriptionDefaultStatusColor:Slate Color Structure)
   n2 = Get SubscriptionActiveStatusColor() -> (SubscriptionActiveStatusColor:Slate Color Structure)
   n3 = Get SubscriptionActiveStatusText() -> (SubscriptionActiveStatusText:Text)
   n4 = Get SubscriptionDefaultStatusText() -> (SubscriptionDefaultStatusText:Text)
   > Return Node!(
    execute
    SubInfoText:Text=Select!(
     Inactive:Text=Get SubscriptionDefaultStatusText()
     Active:Text=@n3
     ActiveWithExternalBilling:Text=@n3
     NotAutoRenewable:Text=@n3
     Canceled:Text=@n3
     PaymentProcessError:Text=Get SubscriptionPaymentErrorStatusText() -> (SubscriptionPaymentErrorStatusText:Text)
     PlatformAuthError:Text=@n4
     RewardsDelayed:Text=@n4
     Unavailable:Text
     Unknown:Text=@n4
     Index:EMcpSubscriptionState Enum=Get SubscriptionState() -> (SubscriptionState:EMcpSubscriptionState Enum=Inactive)
    ) -> (ReturnValue:Text)
    SubInfoColor=Select(
     Inactive:Slate Color Structure=Get SubscriptionDefaultStatusColor()
     Active:Slate Color Structure=@n2
     ActiveWithExternalBilling:Slate Color Structure=@n2
     NotAutoRenewable:Slate Color Structure=@n2
     Canceled:Slate Color Structure=@n2
     PaymentProcessError:Slate Color Structure=@n1
     PlatformAuthError:Slate Color Structure=@n1
     RewardsDelayed:Slate Color Structure=@n1
     Unavailable:Slate Color Structure
     Unknown:Slate Color Structure=@n1
     Index=Get SubscriptionState() -> (SubscriptionState=Inactive)
    ) -> (ReturnValue:Slate Color Structure)
   )
  ]}
 GetSubInfoText {nodes [
   n1 = GetSubInfoText!() -> (then, SubscriptionState:EMcpSubscriptionState Enum=Inactive) {is=title K2Node_FunctionEntry}
   > n2 = GetSubInfoConfig(SubscriptionState=@n1) {is=title K2Node_CallFunction}
   > Return Node!(execute, SubInfoText=@n2)
  ]}
 InitializeWidget {nodes [
   InitializeWidget() {is=title K2Node_FunctionEntry}
   > n1 = Sequence!(execute) -> (then_0, then_1, then_2:Exec, then_3:Exec)
   SetGridSlotAlignment!(
    execute=@n1.then_1
    Widget:Widget Object Reference=Get Switcher_BottomInfo() -> (Switcher_BottomInfo:Common Visibility Switcher Object Reference)
    Horizontal:EHorizontalAlignment Enum=Get PriceAndOwnershipHorizontalAlignment() -> (PriceAndOwnershipHorizontalAlignment:EHorizontalAlignment Enum=HAlign_Fill)
    Vertical:EVerticalAlignment Enum=VAlign_Fill
   ) -> (then)
   > UpdateOwnershipText!(execute, self, OwnedText:Text=Get OwnedText() -> (OwnedText:Text)) -> (then)
   > SetRenderTransformAngle!(
    execute
    self:Widget Object Reference=Get Image_Strike() -> (Image_Strike:Image Object Reference)
    Angle:float=Get StrikeLineRotateAngle() -> (StrikeLineRotateAngle:double)
   ) -> (then)
   > SetBrushFromMaterial(self=Get Image_Strike(), Material=Get StrikeLineMaterial() -> (StrikeLineMaterial:Material Interface Object Reference))
   > n2 = Branch(Condition=Get ShowCurrencyIcon() -> (ShowCurrencyIcon:bool))
   n3 = Get Image_Currency()
   > Set Visibility(Visibility=SelfHitTestInvisible, self=@n3)
   > Set ColorAndOpacity!(
    execute
    ColorAndOpacity:Linear Color Structure=Get CurrencyIconColor() -> (CurrencyIconColor:Linear Color Structure)
    self:Image Object Reference=Get Image_Currency()
   ) -> (then, Output_Get:Linear Color Structure)
   Set Visibility(execute=@n2.else, Visibility=Collapsed, self=@n3)
   SetRealPriceWidgetAndPosition(execute=@n1.then_0)
   > SetRealPriceText()
   > Branch(Condition=Get HasRealPrice())
   > SetFont(
    self=Get RealPriceWidget() -> (RealPriceWidget:Common Text Object Reference)
    InFontInfo=Get RealPriceTextFont() -> (RealPriceTextFont:Slate Font Info Structure)
   )
   > SetRealPriceTextColor!(execute, self, Color:Linear Color Structure=Get RealPriceTextColor() -> (RealPriceTextColor:Linear Color Structure)) -> (then)
   > Set TextTransformPolicy!(
    execute
    TextTransformPolicy:ETextTransformPolicy Enum=Get RealPriceTransformPolicy() -> (RealPriceTransformPolicy:ETextTransformPolicy Enum=None)
    self:Text Object Reference=Get RealPriceWidget()
   ) -> (then, Output_Get:ETextTransformPolicy Enum=None)
   SetAdditionalPriceText(execute=@n1.then_3)
   > SetAdditionalRegularPriceText()
   SetFont(execute=@n1.then_2, self=Get Text_Price(), InFontInfo=Get PriceTextFont() -> (PriceTextFont:Slate Font Info Structure))
   > SetPriceTextColor!(execute, self, Color:Linear Color Structure=Get PriceTextColor() -> (PriceTextColor:Linear Color Structure)) -> (then)
   > SetFont(self=Get Text_RegularPrice(), InFontInfo=Get RegularPriceTextFont() -> (RegularPriceTextFont:Slate Font Info Structure))
   > SetRegularPriceTextColor!(execute, self, Color:Linear Color Structure=Get RegularPriceTextColor() -> (RegularPriceTextColor:Linear Color Structure)) -> (then)
   > SetFont(
    self=Get Text_Owned() -> (Text_Owned:Common Text Object Reference)
    InFontInfo=Get OwnedTextFont() -> (OwnedTextFont:Slate Font Info Structure)
   )
   > SetOwnedIndicatorTextColor!(execute, self, Color:Linear Color Structure=Get OwnedIndicatorTextColor() -> (OwnedIndicatorTextColor:Linear Color Structure)) -> (then)
   > SetFont(
    self=Get Text_SubInfo() -> (Text_SubInfo:Common Text Object Reference)
    InFontInfo=Get SpecialOfferTextFont() -> (SpecialOfferTextFont:Slate Font Info Structure)
   )
   > SetSpecialOfferTextColor!(execute, self, Color:Linear Color Structure=Get SpecialOfferTextColor() -> (SpecialOfferTextColor:Linear Color Structure)) -> (then)
   > SetFont(self=Get Text_AdditionalPriceText(), InFontInfo=Get AdditionalPriceTextFont() -> (AdditionalPriceTextFont:Slate Font Info Structure))
   > SetAdditionalPriceTextColor!(execute, self, Color:Linear Color Structure=Get AdditionalPriceTextColor() -> (AdditionalPriceTextColor:Linear Color Structure)) -> (then)
   > SetFont(
    self=Get Text_AdditionalRegularPriceText() -> (Text_AdditionalRegularPriceText:Common Text Object Reference)
    InFontInfo=Get AdditionalRegularPriceTextFont() -> (AdditionalRegularPriceTextFont:Slate Font Info Structure)
   )
   > SetAdditionalRegularPriceTextColor!(
    execute
    self
    Color:Linear Color Structure=Get AdditionalRegularPriceTextColor() -> (AdditionalRegularPriceTextColor:Linear Color Structure)
   ) -> (then)
  ]}
 SetAdditionalPriceText {nodes [
   n1 = SetAdditionalPriceText!() -> (then, Text:Text) {is=title K2Node_FunctionEntry}
   > "SetText (Text)"(self=Get Text_AdditionalPriceText(), InText=@n1)
  ]}
 SetAdditionalPriceTextColor {nodes [
   n1 = SetAdditionalPriceTextColor!() -> (then, Color:Linear Color Structure) {is=title K2Node_FunctionEntry}
   > Set Text Color!(execute:Exec, TextWidget:Text Object Reference=Get Text_AdditionalPriceText(), Color:Linear Color Structure=@n1) -> (then:Exec) {is=title K2Node_MacroInstance}
  ]}
 SetAdditionalRegularPriceText {nodes [
   n1 = SetAdditionalRegularPriceText!() -> (then, Text:Text) {is=title K2Node_FunctionEntry}
   > "SetText (Text)"(self=Get Text_AdditionalRegularPriceText(), InText=@n1)
  ]}
 SetAdditionalRegularPriceTextColor {nodes [
   n1 = SetAdditionalRegularPriceTextColor!() -> (then, Color:Linear Color Structure) {is=title K2Node_FunctionEntry}
   > Set Text Color(TextWidget=Get Text_AdditionalRegularPriceText(), Color=@n1)
  ]}
 SetColorToAll {nodes [
   n1 = SetColorToAll!() -> (then, Color:Linear Color Structure) {is=title K2Node_FunctionEntry}
   n2 = Alpha Override!(LinearColor:"Linear Color Structure (by ref)"=@n1, NewAlpha:float=0.5) -> (ReturnValue:Linear Color Structure) {is=title K2Node_MacroInstance}
   > SetRegularPriceTextColor(Color=@n2)
   > SetColorAndOpacity(
    self:Image Object Reference=Get Image_Currency_Regular() -> (Image_Currency_Regular:Image Object Reference)
    InColorAndOpacity:Linear Color Structure=@n2
   )
   > SetOwnedIndicatorTextColor(Color=@n1)
   > SetPriceTextColor(Color=@n1)
   > SetAdditionalPriceTextColor(Color=@n1) {is=title K2Node_CallFunction}
   > SetAdditionalRegularPriceTextColor(Color=@n1) {is=title K2Node_CallFunction}
   > SetSpecialOfferTextColor(Color=@n1)
   > Set Strike Color!(execute:Exec, Color:Linear Color Structure=Alpha Override(LinearColor=@n1, NewAlpha=0.4)) -> (then:Exec) {is=title K2Node_MacroInstance}
   > Set ColorAndOpacity(ColorAndOpacity=@n1, self=Get Image_Currency())
   > SetRealPriceTextColor(Color=@n1)
  ]}
 SetCurrency {nodes [
   n1 = SetCurrency!() -> (then, Currency:EFortPriceCurrency Enum=Mtx) {is=title K2Node_FunctionEntry}
   > UpdateCurrency!(execute, self:WBP Price Platform Store Icon Object Reference=Get WBP_Price_PlatformStoreIcon(), CurrencyType:EFortPriceCurrency Enum=@n1) -> (then)
   n2 = Get Image_Currency()
   n3 = Get!(self:Image Object Reference=@n2) -> (
    Brush_TintColor:Slate Color Structure
    Brush_DrawAs:ESlateBrushDrawType Enum
    Brush_Tiling:ESlateBrushTileType Enum
    Brush_Mirroring:ESlateBrushMirrorType Enum
    Brush_ImageSize:Vector 2D Structure
    Brush_Margin:Margin Structure
    Brush_ResourceObject:Object Reference
    Brush_OutlineSettings:Slate Brush Outline Settings Structure
   )
   > SetBrush!(
    execute
    self=@n2
    InBrush_TintColor:Slate Color Structure=@n3
    InBrush_DrawAs:ESlateBrushDrawType Enum=@n3.Brush_DrawAs
    InBrush_Tiling:ESlateBrushTileType Enum=@n3.Brush_Tiling
    InBrush_Mirroring:ESlateBrushMirrorType Enum=@n3.Brush_Mirroring
    InBrush_ImageSize:Vector 2D Structure=Get CurrencyImageSize() -> (CurrencyImageSize:Vector 2D Structure)
    InBrush_Margin:Margin Structure=@n3.Brush_Margin
    InBrush_ResourceObject:Object Reference=Select!(
     Mtx:Object Reference=Get VBucksIcon() -> (VBucksIcon:Texture 2D Object Reference)
     RealMoney:Object Reference=Get RealMoneyIcon() -> (RealMoneyIcon:Texture 2D Object Reference)
     Index:EFortPriceCurrency Enum=Get Currency() -> (Currency:EFortPriceCurrency Enum=Mtx)
    ) -> (ReturnValue:Object Reference)
    InBrush_OutlineSettings:Slate Brush Outline Settings Structure=@n3.Brush_OutlineSettings
   ) -> (then)
   n4 = NOT Boolean(A=IsValid(Object=Get RealMoneyIcon()))
   n5 = "Equal (Enum)"(A:EFortPriceCurrency Enum=Get Currency() -> (Currency=Mtx), B:EFortPriceCurrency Enum=RealMoney)
   > SetVisibility(
    self=Get Image_Currency()
    InVisibility=Select!(Option 0:ESlateVisibility Enum=SelfHitTestInvisible, Option 1:ESlateVisibility Enum=Collapsed, Index:bool=AND Boolean(A=@n5, B=@n4)) -> (ReturnValue:ESlateVisibility Enum)
   )
  ]}
 SetOwnedIndicatorTextColor {nodes [
   n1 = SetOwnedIndicatorTextColor!() -> (then, Color:Linear Color Structure) {is=title K2Node_FunctionEntry}
   > Set Text Color(TextWidget=Get Text_Owned(), Color=@n1)
  ]}
 SetPrice {nodes [
   n1 = SetPrice!() -> (then, Price:Text) {is=title K2Node_FunctionEntry}
   > "SetText (Text)"(self=Get Text_Price(), InText=@n1)
  ]}
 SetPriceTextColor {nodes [
   n1 = SetPriceTextColor!() -> (then, Color:Linear Color Structure) {is=title K2Node_FunctionEntry}
   > Set Text Color(TextWidget=Get Text_Price(), Color=@n1)
  ]}
 SetRealPriceText {nodes [
   n1 = SetRealPriceText!() -> (then, Text:Text) {is=title K2Node_FunctionEntry}
   > "SetText (Text)"(self=Get RealPriceWidget(), InText=@n1)
   n2 = TextIsEmpty(InText=Get Text!() -> (Text))
   > SetWidgetHeight!(execute, self, IsRealPriceShowing:bool=NOT Boolean(A=@n2)) -> (then)
   > n3 = Branch(Condition=@n2)
   SetVisibility(execute=@n3.else, self=Get RealPriceWidget(), InVisibility=SelfHitTestInvisible)
  ]}
 SetRealPriceTextColor {nodes [
   n1 = SetRealPriceTextColor!() -> (then, Color:Linear Color Structure) {is=title K2Node_FunctionEntry}
   > Set Text Color(TextWidget=Get RealPriceWidget(), Color=@n1)
  ]}
 SetRealPriceWidgetAndPosition {nodes [
   SetRealPriceWidgetAndPosition() {is=title K2Node_FunctionEntry}
   > n1 = Is Valid(InputObject=Get RealPriceWidget())
   n2 = Get ShowRealPriceOnTheSide() -> (ShowRealPriceOnTheSide:bool)
   Set Grid Slot Column and Row!(
    execute:Exec=@n1.Is Not Valid
    Widget:Widget Object Reference=Get Text_RealPrice()
    Row:int=Select(Option 0:int=1, Option 1:int=0, Index=@n2) -> (ReturnValue:int)
    RowSpan:int
    Column:int=Select(Option 0=0, Option 1=1, Index=@n2)
    ColumnSpan:int
   ) -> (then:Exec) {is=title K2Node_MacroInstance}
   > SetGridSlotAlignment(
    Widget=Get Text_RealPrice()
    Horizontal=Get RealPriceHorizontalAlignment() -> (RealPriceHorizontalAlignment:EHorizontalAlignment Enum=HAlign_Fill)
    Vertical=Get RealPriceVerticalAlignment() -> (RealPriceVerticalAlignment:EVerticalAlignment Enum=VAlign_Fill)
   )
   > Set RealPriceWidget(RealPriceWidget:Common Text Object Reference=Get Text_RealPrice()) -> (Output_Get:Common Text Object Reference)
   SetVisibility(execute=@n1.Is Valid, self=Get Text_RealPrice(), InVisibility=Collapsed)
  ]}
 SetRegularPrice {nodes [
   n1 = SetRegularPrice!() -> (then, Price:Text) {is=title K2Node_FunctionEntry}
   > "SetText (Text)"(self=Get Text_RegularPrice(), InText=@n1)
  ]}
 SetRegularPriceImageVisibility {nodes [
   n1 = SetRegularPriceImageVisibility!() -> (then, InVisibility:ESlateVisibility Enum=Visible) {is=title K2Node_FunctionEntry}
   > SetVisibility(self=Get Image_Currency_Regular(), InVisibility=@n1)
  ]}
 SetRegularPriceTextColor {nodes [
   n1 = SetRegularPriceTextColor!() -> (then, Color:Linear Color Structure) {is=title K2Node_FunctionEntry}
   > Set Text Color(TextWidget=Get Text_RegularPrice(), Color=@n1)
   > Set Strike Color(Color=Alpha Override(LinearColor=@n1, NewAlpha=0.4))
  ]}
 SetSpecialOfferTextColor {nodes [
   n1 = SetSpecialOfferTextColor!() -> (then, Color:Linear Color Structure) {is=title K2Node_FunctionEntry}
   > Set Text Color(TextWidget=Get Text_SubInfo(), Color=@n1)
  ]}
 SetStrikeColor {nodes [
   n1 = Inputs!() -> (execute:Exec, Color:Linear Color Structure)
   > n2 = GetDynamicMaterial(self=Get Image_Strike())
   > SetVectorParameterValue(self=@n2, ParameterName=Color_Idle, Value=@n1)
   > SetVectorParameterValue(self=@n2, ParameterName=Color_Hovered, Value=@n1)
   > SetVectorParameterValue(self=@n2, ParameterName=Color_Focused, Value=@n1)
   > SetVectorParameterValue(self=@n2, ParameterName=Color_Pressed, Value=@n1)
   > Outputs!(then:Exec)
  ]}
 SetStrikethroughPriceVisibility {nodes [
   n1 = SetStrikethroughPriceVisibility!() -> (then, Visibility:ESlateVisibility Enum=Visible) {is=title K2Node_FunctionEntry}
   > SetVisibility(self=Get Overlay_RegularPrice() -> (Overlay_RegularPrice:Overlay Object Reference), InVisibility=@n1)
  ]}
 SetSubInfoColor {nodes [
   n1 = SetSubInfoColor!() -> (then, SubInfoColor:Slate Color Structure) {is=title K2Node_FunctionEntry}
   > SetColorAndOpacity(self:Text Object Reference=Get Text_SubInfo(), InColorAndOpacity:Slate Color Structure=@n1)
  ]}
 SetSubInfoText {nodes [
   n1 = SetSubInfoText!() -> (then, SubInfoText:Text) {is=title K2Node_FunctionEntry}
   > "SetText (Text)"(self=Get Text_SubInfo(), InText=@n1)
  ]}
 SetSwitcherState {nodes [
   n1 = SetSwitcherState!() -> (then, PriceState:E_PriceSwitcherState Enum=NewEnumerator0) {is=title K2Node_FunctionEntry}
   > SetActiveWidget(
    self=Get Switcher_BottomInfo()
    Widget=Select!(
     NewEnumerator0:Overlay Object Reference=Get SwitcherState_NotOwned() -> (SwitcherState_NotOwned:Overlay Object Reference)
     NewEnumerator1:Overlay Object Reference=Get SwitcherState_Owned() -> (SwitcherState_Owned:Overlay Object Reference)
     NewEnumerator2:Overlay Object Reference=Get SwitcherState_SubscriptionInfo() -> (SwitcherState_SubscriptionInfo:Overlay Object Reference)
     Index:E_PriceSwitcherState Enum=@n1
    ) -> (ReturnValue:Overlay Object Reference)
   )
  ]}
 SetTextColor {nodes [
   n1 = Inputs!() -> (execute, TextWidget:Text Object Reference, Color)
   > n2 = GetDynamicFontMaterial!(execute, self:Text Object Reference=@n1) -> (then, ReturnValue:Material Instance Dynamic Object Reference)
   > SetVectorParameterValue(self=@n2, ParameterName=Color_Idle, Value=@n1.Color)
   > SetVectorParameterValue(self=@n2, ParameterName=Color_Hovered, Value=@n1.Color)
   > SetVectorParameterValue(self=@n2, ParameterName=Color_Focused, Value=@n1.Color)
   > SetVectorParameterValue(self=@n2, ParameterName=Color_Pressed, Value=@n1.Color)
   > Outputs()
  ]}
 SetWidgetHeight {nodes [
   n1 = SetWidgetHeight!() -> (then, IsRealPriceShowing:bool=~) {is=title K2Node_FunctionEntry}
   > SetHeightOverride(
    self=Get SizeBox_PriceStateSwitcher() -> (SizeBox_PriceStateSwitcher:Size Box Object Reference)
    InHeightOverride=Select!(
     Option 0:float=Get WidgetHeight() -> (WidgetHeight:float)
     Option 1:float=Get WidgetHeightWithRealPrice() -> (WidgetHeightWithRealPrice:float)
     Index:bool=@n1
    ) -> (ReturnValue:float)
   )
  ]}
 UpdateOwnershipText {nodes [
   n1 = UpdateOwnershipText!() -> (then, OwnedText:Text) {is=title K2Node_FunctionEntry}
   > n2 = Set OwnedText(OwnedText:Text=@n1) -> (Output_Get:Text)
   > "SetText (Text)"(self=Get Text_Owned(), InText=@n2)
  ]}
]
variables [
 AdditionalPriceTextColor {type=LinearColor, default=<rgba 1 1 1 1>}
 AdditionalPriceTextFont {
  type=SlateFontInfo
  default {
   bForceMonospaced=false
   bMaterialIsStencil=false
   fontMaterial=/FortUILibrary/Systems/ItemShopTile/Materials/PriceSwitcher/MI_UIKit_PriceSwitcher_Text
   fontObject=&HeadingNow
   letterSpacing=0
   monospacedWidth=1
   size=10.5
   skewAmount=0
   typefaceFontName=74 Regular
   outlineSettings {
    bApplyOutlineToDropShadows=false
    bMiteredCorners=false
    bSeparateFillAlpha=false
    outlineMaterial=null
    outlineSize=0
    outlineColor=<rgba 0 0 0 1>
   }
  }
 }
 AdditionalPriceTextMobilePadding {type=Margin, default=<ltrb 0 0 4 1>}
 AdditionalPriceTextPadding {type=Margin, default=<ltrb 0 0 4 1.2>}
 AdditionalRegularPriceTextColor {type=LinearColor, default=<rgba 1 1 1 1>}
 AdditionalRegularPriceTextFont {
  type=SlateFontInfo
  default {
   bForceMonospaced=false
   bMaterialIsStencil=false
   fontMaterial=&MI_UIKit_PriceSwitcher_Text
   fontObject=&HeadingNow
   letterSpacing=0
   monospacedWidth=1
   size=15.75
   skewAmount=0
   typefaceFontName=75 Medium
   outlineSettings {
    bApplyOutlineToDropShadows=false
    bMiteredCorners=false
    bSeparateFillAlpha=false
    outlineMaterial=null
    outlineSize=0
    outlineColor=<rgba 0 0 0 1>
   }
  }
 }
 CurrencyIconColor {type=LinearColor, default=<rgba 1 1 1 1>}
 CurrencyImageSize {type=Vector2D, default=<24 24>}
 CurrencyRegularVisibility {type=ESlateVisibility, default=Collapsed}
 DefaultCurrencyImageScale {type=Vector2D, default=<1 1>}
 DefaultPricePadding {type=Margin, default=<ltrb 4 0 4 0>}
 DefaultRealPricePadding {type=Margin, default=<ltrb 0 4 0 0>}
 DefaultRegularPricePadding {type=Margin, default=<ltrb 4 0 4 0>}
 HasRealPrice {type=boolean}
 IsAdditionalRealPrice {type=boolean}
 IsPriceShowing {type=boolean, default=true}
 MobileCurrencyImageScale {type=Vector2D, default=<1.3 1.3>}
 MobilePricePadding {type=Margin, default=<ltrb 10 0 4 0>}
 MobileRealPricePadding {type=Margin, default=<ltrb 0 0 0 0>}
 MobileRegularPricePadding {type=Margin, default=<ltrb 4 0 4 0>}
 OwnedIndicatorTextColor {type=LinearColor, default=<rgba 0.0369 1 0.2232 1>}
 OwnedText {type=string, default=OWNED}
 OwnedTextFont {
  type=SlateFontInfo
  default {
   bForceMonospaced=false
   bMaterialIsStencil=false
   fontMaterial=&MI_UIKit_PriceSwitcher_Text
   fontObject=&HeadingNow
   letterSpacing=0
   monospacedWidth=1
   size=15.75
   skewAmount=0
   typefaceFontName=75 Medium
   outlineSettings {
    bApplyOutlineToDropShadows=false
    bMiteredCorners=false
    bSeparateFillAlpha=false
    outlineMaterial=null
    outlineSize=0
    outlineColor=<rgba 0 0 0 1>
   }
  }
 }
 PreviewAdditionalPriceText {type=string, default=VAT incl.}
 PreviewAdditionalRegularPriceText {type=string, default=*}
 PreviewCurrency {type=EFortPriceCurrency, default=Mtx}
 PreviewMobileState {type=boolean}
 PreviewPlayStationPlatform {type=boolean}
 PreviewPrice {type=integer, default=1800}
 PreviewPriceState {type=E_PriceSwitcherState, default=Default}
 PreviewRealPriceText {type=string, default=(equivalent to $8.99*)}
 PreviewRegularPrice {type=integer, default=2000}
 PriceAndOwnershipHorizontalAlignment {type=EHorizontalAlignment, default=HAlign_Fill}
 PriceDisplayMode {type=E_PriceDisplayMode, default=PriceAndOwnership}
 PriceTextColor {type=LinearColor, default=<rgba 1 1 1 1>}
 PriceTextFont {
  type=SlateFontInfo
  default {
   bForceMonospaced=false
   bMaterialIsStencil=false
   fontMaterial=&MI_UIKit_PriceSwitcher_Text
   fontObject=&HeadingNow
   letterSpacing=0
   monospacedWidth=1
   size=15.75
   skewAmount=0
   typefaceFontName=75 Medium
   outlineSettings {
    bApplyOutlineToDropShadows=false
    bMiteredCorners=false
    bSeparateFillAlpha=false
    outlineMaterial=null
    outlineSize=0
    outlineColor=<rgba 0 0 0 1>
   }
  }
 }
 RealMoneyIcon {type=Texture2D}
 RealPriceHorizontalAlignment {type=EHorizontalAlignment, default=HAlign_Left}
 RealPriceTextColor {type=LinearColor, default=<rgba 1 1 1 1>}
 RealPriceTextFont {
  type=SlateFontInfo
  default {
   bForceMonospaced=false
   bMaterialIsStencil=false
   fontMaterial=&MI_UIKit_PriceSwitcher_Text
   fontObject=&HeadingNow
   letterSpacing=0
   monospacedWidth=1
   size=10.5
   skewAmount=0
   typefaceFontName=74 Regular
   outlineSettings {
    bApplyOutlineToDropShadows=false
    bMiteredCorners=false
    bSeparateFillAlpha=false
    outlineMaterial=null
    outlineSize=0
    outlineColor=<rgba 0 0 0 1>
   }
  }
 }
 RealPriceTransformPolicy {type=ETextTransformPolicy, default=None}
 RealPriceVerticalAlignment {type=EVerticalAlignment, default=VAlign_Fill}
 RealPriceWidget {type=CommonTextBlock, default=null}
 RegularPriceTextColor {type=LinearColor, default=<rgba 1 1 1 0.5>}
 RegularPriceTextFont {
  type=SlateFontInfo
  default {
   bForceMonospaced=false
   bMaterialIsStencil=false
   fontMaterial=&MI_UIKit_PriceSwitcher_Text
   fontObject=&HeadingNow
   letterSpacing=0
   monospacedWidth=1
   size=15.75
   skewAmount=0
   typefaceFontName=75 Medium
   outlineSettings {
    bApplyOutlineToDropShadows=false
    bMiteredCorners=false
    bSeparateFillAlpha=false
    outlineMaterial=null
    outlineSize=0
    outlineColor=<rgba 0 0 0 1>
   }
  }
 }
 ShowCurrencyIcon {type=boolean}
 ShowRealPriceOnTheSide {type=boolean}
 SpecialOfferTextColor {type=LinearColor, default=<rgba 1 1 1 1>}
 SpecialOfferTextFont {
  type=SlateFontInfo
  default {
   bForceMonospaced=false
   bMaterialIsStencil=false
   fontMaterial=&MI_UIKit_PriceSwitcher_Text
   fontObject=&HeadingNow
   letterSpacing=0
   monospacedWidth=1
   size=15.75
   skewAmount=0
   typefaceFontName=75 Medium
   outlineSettings {
    bApplyOutlineToDropShadows=false
    bMiteredCorners=false
    bSeparateFillAlpha=false
    outlineMaterial=null
    outlineSize=0
    outlineColor=<rgba 0 0 0 1>
   }
  }
 }
 StrikeColor {type=LinearColor, default=<rgba 1 1 1 0.4>}
 StrikeLineMaterial {type=MaterialInterface, default=/FortUILibrary/Systems/ItemShopTile/Materials/PriceSwitcher/MI_UIKit_ItemShopTile_RegularPriceStrike}
 StrikeLineRotateAngle {type=number, default=-3}
 SubscriptionActiveStatusColor {type=SlateColor, default {colorUseRule=UseColor_Specified, specifiedColor=<rgba 0.0369 1 0.2232 1>}}
 SubscriptionActiveStatusText {type=string, default=Active Subscription}
 SubscriptionDefaultStatusColor {type=SlateColor, default {colorUseRule=UseColor_Specified, specifiedColor=<rgba 1 1 1 1>}}
 SubscriptionDefaultStatusText {type=string, default=Learn More}
 SubscriptionPaymentErrorStatusText {type=string, default=Payment Process Issue}
 VBucksIcon {type=Texture2D, default=/FortUILibrary/Textures/Icons/T_UI_Icon_UIKit_VBuck}
 WidgetHeight {type=number, default=28}
 WidgetHeightWithRealPrice {type=number, default=46}
]
widgets {
 info {inheritedWidgetCount=0, parentClass=&UserWidget, rootWidgetClass=&SizeBox, widgetCount=23}
 widgets [
  SizeBox_PriceStateSwitcher {
   widgetClassPath=&SizeBox
   bIsVariable=true
   widget=SizeBox_PriceStateSwitcher
   properties {bOverride_HeightOverride=true, heightOverride=46, renderTransformPivot=<0.5 1>}
  }
  GridPanel_0 {
   widgetClassPath=&GridPanel
   parent=SizeBox_PriceStateSwitcher
   widget=GridPanel_0
   properties {bOverride_Cursor=true}
   slot {class=SizeBoxSlot}
  }
  Switcher_BottomInfo {
   widgetClassPath=&CommonVisibilitySwitcher
   bIsVariable=true
   parent=GridPanel_0
   widget=Switcher_BottomInfo
   slot {class=GridSlot}
  }
  SwitcherState_NotOwned {
   widgetClassPath=&Overlay
   bIsVariable=true
   parent=Switcher_BottomInfo
   widget=SwitcherState_NotOwned
   slot {class=CommonVisibilitySwitcherSlot}
  }
  HB_PriceContainer {
   widgetClassPath=&HorizontalBox
   bIsVariable=true
   parent=SwitcherState_NotOwned
   widget=HB_PriceContainer
   slot {class=OverlaySlot, properties {horizontalAlignment=HAlign_Fill, verticalAlignment=VAlign_Center}}
  }
  Overlay_0 {
   widgetClassPath=&Overlay
   parent=HB_PriceContainer
   widget=Overlay_0
   slot {class=HorizontalBoxSlot, properties {horizontalAlignment=HAlign_Left, verticalAlignment=VAlign_Center}}
  }
  WBP_Price_PlatformStoreIcon {
   widgetClassPath=&/WBP_Price_PlatformStoreIcon
   bIsVariable=true
   parent=Overlay_0
   widget=WBP_Price_PlatformStoreIcon
   properties {currency=EFortPriceCurrency_MAX, previewPlayStationPlatform=true}
   slot {class=OverlaySlot, properties {horizontalAlignment=HAlign_Center, verticalAlignment=VAlign_Center}}
  }
  Image_Currency {
   widgetClassPath=&Image
   bIsVariable=true
   parent=Overlay_0
   widget=Image_Currency
   properties {brush {
     bIsDynamicallyLoaded=false
     drawAs=Image
     imageType=NoImage
     mirroring=NoMirror
     resourceName=None
     resourceObject=null
     tiling=NoTile
     imageSize=<24 24>
     margin=<ltrb 0 0 0 0>
     outlineSettings {
      bUseBrushTransparency=false
      roundingType=HalfHeightRadius
      width=0
      color {colorUseRule=UseColor_Specified, specifiedColor=<rgba 0 0 0 0>}
      cornerRadii=<0 0 0 0>
     }
     tintColor {colorUseRule=UseColor_Specified, specifiedColor=<rgba 1 1 1 1>}
     uVRegion {bIsValid=false, max=<0 0>, min=<0 0>}
    }}
   slot {class=OverlaySlot, properties {horizontalAlignment=HAlign_Center, verticalAlignment=VAlign_Center}}
  }
  Spacer_498 {
   widgetClassPath=&Spacer
   bIsVariable=true
   parent=HB_PriceContainer
   widget=Spacer_498
   slot {class=HorizontalBoxSlot, properties {horizontalAlignment=HAlign_Left}}
  }
  Text_Price {
   widgetClassPath=&CommonTextBlock
   bIsVariable=true
   parent=HB_PriceContainer
   widget=Text_Price
   properties {
    justification=Center
    mobileFontSizeMultiplier=1.2
    text=2000
    textTransformPolicy=ToUpper
    font {
     bForceMonospaced=false
     bMaterialIsStencil=false
     fontMaterial=null
     fontObject=&HeadingNow
     letterSpacing=0
     monospacedWidth=1
     size=15.75
     skewAmount=0
     typefaceFontName=75 Medium
     outlineSettings {
      bApplyOutlineToDropShadows=false
      bMiteredCorners=false
      bSeparateFillAlpha=false
      outlineMaterial=null
      outlineSize=0
      outlineColor=<rgba 0 0 0 1>
     }
    }
    shadowOffset=<0 0>
   }
   slot {class=HorizontalBoxSlot, properties {horizontalAlignment=HAlign_Right, verticalAlignment=VAlign_Center, padding=<ltrb 4 0 4 0>}}
  }
  Text_AdditionalPriceText {
   widgetClassPath=&CommonTextBlock
   bIsVariable=true
   parent=HB_PriceContainer
   widget=Text_AdditionalPriceText
   properties {
    applyLineHeightToBottomLine=false
    bAutoCollapseWithEmptyText=true
    justification=Center
    mobileFontSizeMultiplier=1.2
    text=VAT incl.
    font {
     bForceMonospaced=false
     bMaterialIsStencil=false
     fontMaterial=null
     fontObject=&HeadingNow
     letterSpacing=0
     monospacedWidth=1
     size=10.5
     skewAmount=0
     typefaceFontName=74 Regular
     outlineSettings {
      bApplyOutlineToDropShadows=false
      bMiteredCorners=false
      bSeparateFillAlpha=false
      outlineMaterial=null
      outlineSize=0
      outlineColor=<rgba 0 0 0 1>
     }
    }
    shadowOffset=<0 0>
   }
   slot {class=HorizontalBoxSlot, properties {horizontalAlignment=HAlign_Right, verticalAlignment=VAlign_Bottom, padding=<ltrb 0 0 4 0>}}
  }
  Spacer {
   widgetClassPath=&Spacer
   bIsVariable=true
   parent=HB_PriceContainer
   widget=Spacer
   properties {size=<2 1>}
   slot {class=HorizontalBoxSlot, properties {horizontalAlignment=HAlign_Left}}
  }
  Overlay_RegularPrice {
   widgetClassPath=&Overlay
   bIsVariable=true
   parent=HB_PriceContainer
   widget=Overlay_RegularPrice
   properties {visibility=Collapsed}
   slot {class=HorizontalBoxSlot, properties {horizontalAlignment=HAlign_Right, verticalAlignment=VAlign_Center}}
  }
  RegularHorizontal {
   widgetClassPath=&HorizontalBox
   bIsVariable=true
   parent=Overlay_RegularPrice
   widget=RegularHorizontal
   slot {class=OverlaySlot, properties {horizontalAlignment=HAlign_Fill, verticalAlignment=VAlign_Center, padding=<ltrb 4 0 4 0>}}
  }
  Image_Currency_Regular {
   widgetClassPath=&Image
   bIsVariable=true
   parent=RegularHorizontal
   widget=Image_Currency_Regular
   properties {
    visibility=Collapsed
    brush {
     bIsDynamicallyLoaded=false
     drawAs=Image
     imageType=NoImage
     mirroring=NoMirror
     resourceName=None
     resourceObject=&T_UI_Icon_UIKit_VBuck
     tiling=NoTile
     imageSize=<24 24>
     margin=<ltrb 0 0 0 0>
     outlineSettings {
      bUseBrushTransparency=false
      roundingType=HalfHeightRadius
      width=0
      color {colorUseRule=UseColor_Specified, specifiedColor=<rgba 0 0 0 0>}
      cornerRadii=<0 0 0 0>
     }
     tintColor {colorUseRule=UseColor_Specified, specifiedColor=<rgba 1 1 1 1>}
     uVRegion {bIsValid=false, max=<0 0>, min=<0 0>}
    }
   }
   slot {class=HorizontalBoxSlot, properties {horizontalAlignment=HAlign_Center, verticalAlignment=VAlign_Center}}
  }
  Text_RegularPrice {
   widgetClassPath=&CommonTextBlock
   bIsVariable=true
   parent=RegularHorizontal
   widget=Text_RegularPrice
   properties {
    justification=Center
    mobileFontSizeMultiplier=1.2
    text=2000
    textTransformPolicy=ToUpper
    font {
     bForceMonospaced=false
     bMaterialIsStencil=false
     fontMaterial=null
     fontObject=&HeadingNow
     letterSpacing=0
     monospacedWidth=1
     size=15.75
     skewAmount=0
     typefaceFontName=75 Medium
     outlineSettings {
      bApplyOutlineToDropShadows=false
      bMiteredCorners=false
      bSeparateFillAlpha=false
      outlineMaterial=null
      outlineSize=0
      outlineColor=<rgba 0 0 0 1>
     }
    }
    shadowOffset=<0 0>
   }
   slot {class=HorizontalBoxSlot}
  }
  Image_Strike {
   widgetClassPath=&Image
   bIsVariable=true
   parent=Overlay_RegularPrice
   widget=Image_Strike
   properties {
    visibility=SelfHitTestInvisible
    brush {
     bIsDynamicallyLoaded=false
     drawAs=Image
     imageType=NoImage
     mirroring=NoMirror
     resourceName=None
     resourceObject=&MI_UIKit_ItemShopTile_RegularPriceStrike
     tiling=NoTile
     imageSize=<32 4>
     margin=<ltrb 0 0 0 0>
     outlineSettings {
      bUseBrushTransparency=false
      roundingType=HalfHeightRadius
      width=0
      color {colorUseRule=UseColor_Specified, specifiedColor=<rgba 0 0 0 0>}
      cornerRadii=<0 0 0 0>
     }
     tintColor {colorUseRule=UseColor_Specified, specifiedColor=<rgba 1 1 1 1>}
     uVRegion {bIsValid=false, max=<0 0>, min=<0 0>}
    }
    renderTransform {angle=-3, scale=<1 1>, shear=<0 0>, translation=<0 0>}
   }
   slot {class=OverlaySlot, properties {horizontalAlignment=HAlign_Fill, verticalAlignment=VAlign_Center}}
  }
  Text_AdditionalRegularPriceText {
   widgetClassPath=&CommonTextBlock
   bIsVariable=true
   parent=HB_PriceContainer
   widget=Text_AdditionalRegularPriceText
   properties {
    bAutoCollapseWithEmptyText=true
    justification=Center
    mobileFontSizeMultiplier=1.2
    text=*
    font {
     bForceMonospaced=false
     bMaterialIsStencil=false
     fontMaterial=null
     fontObject=&HeadingNow
     letterSpacing=0
     monospacedWidth=1
     size=15.75
     skewAmount=0
     typefaceFontName=75 Medium
     outlineSettings {
      bApplyOutlineToDropShadows=false
      bMiteredCorners=false
      bSeparateFillAlpha=false
      outlineMaterial=null
      outlineSize=0
      outlineColor=<rgba 0 0 0 1>
     }
    }
    shadowOffset=<0 0>
   }
   slot {class=HorizontalBoxSlot, properties {horizontalAlignment=HAlign_Right, verticalAlignment=VAlign_Center, padding=<ltrb 4 0 4 0>}}
  }
  SwitcherState_Owned {
   widgetClassPath=&Overlay
   bIsVariable=true
   parent=Switcher_BottomInfo
   widget=SwitcherState_Owned
   slot {class=CommonVisibilitySwitcherSlot}
  }
  Text_Owned {
   widgetClassPath=&CommonTextBlock
   bIsVariable=true
   parent=SwitcherState_Owned
   widget=Text_Owned
   properties {
    justification=Center
    text=Owned
    textTransformPolicy=ToUpper
    font {
     bForceMonospaced=false
     bMaterialIsStencil=false
     fontMaterial=null
     fontObject=&HeadingNow
     letterSpacing=0
     monospacedWidth=1
     size=15.75
     skewAmount=0
     typefaceFontName=75 Medium
     outlineSettings {
      bApplyOutlineToDropShadows=false
      bMiteredCorners=false
      bSeparateFillAlpha=false
      outlineMaterial=null
      outlineSize=0
      outlineColor=<rgba 0 0 0 1>
     }
    }
    shadowOffset=<0 0>
   }
   slot {class=OverlaySlot, properties {verticalAlignment=VAlign_Center}}
  }
  SwitcherState_SubscriptionInfo {
   widgetClassPath=&Overlay
   bIsVariable=true
   parent=Switcher_BottomInfo
   widget=SwitcherState_SubscriptionInfo
   slot {class=CommonVisibilitySwitcherSlot}
  }
  Text_SubInfo {
   widgetClassPath=&CommonTextBlock
   bIsVariable=true
   parent=SwitcherState_SubscriptionInfo
   widget=Text_SubInfo
   properties {
    justification=Center
    mobileFontSizeMultiplier=1.1
    text="[PH] Learn More"
    textTransformPolicy=ToUpper
    font {
     bForceMonospaced=false
     bMaterialIsStencil=false
     fontMaterial=null
     fontObject=&HeadingNow
     letterSpacing=0
     monospacedWidth=1
     size=15.75
     skewAmount=0
     typefaceFontName=75 Medium
     outlineSettings {
      bApplyOutlineToDropShadows=false
      bMiteredCorners=false
      bSeparateFillAlpha=false
      outlineMaterial=null
      outlineSize=0
      outlineColor=<rgba 0 0 0 1>
     }
    }
    shadowOffset=<0 0>
   }
   slot {class=OverlaySlot, properties {verticalAlignment=VAlign_Center}}
  }
  Text_RealPrice {
   widgetClassPath=&CommonTextBlock
   bIsVariable=true
   parent=GridPanel_0
   widget=Text_RealPrice
   properties {
    mobileFontSizeMultiplier=1.2
    text="[ph] (equivalent to $8.99*)"
    font {
     bForceMonospaced=false
     bMaterialIsStencil=false
     fontMaterial=null
     fontObject=&NotoSans
     letterSpacing=0
     monospacedWidth=1
     size=24
     skewAmount=0
     typefaceFontName=NotoSans-Regular
     outlineSettings {
      bApplyOutlineToDropShadows=false
      bMiteredCorners=false
      bSeparateFillAlpha=false
      outlineMaterial=null
      outlineSize=0
      outlineColor=<rgba 0 0 0 1>
     }
    }
    shadowOffset=<0 0>
   }
   slot {class=GridSlot, properties {horizontalAlignment=HAlign_Left, row=1, verticalAlignment=VAlign_Top}}
  }
 ]
}
